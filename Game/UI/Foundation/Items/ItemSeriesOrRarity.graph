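# Item label widget that shows either an item's rarity name or its series name in a widget switcher.
# UpdateItemDefinition collapses it when the definition is invalid, the item view model disallows rarity (shown if no view model), or OnlyShowSeries is set without a series.
# Otherwise it fills both texts, activates the bordered series label when a series exists, and tints both texts from the rarity-or-series palette.
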WidgetBlueprint /Game/UI/Foundation/Items/ItemSeriesOrRarity
graphs [
 EventGraph {nodes [
   Event PreConstruct()
   Event Construct()
   Event Tick()
  ]}
 UpdateItemDefinition {nodes [
   n1 = UpdateItemDefinition!() -> (then, ItemDefinition:Item Definition Base Object Reference, Item:Fort Item VM Object Reference) {is=title K2Node_FunctionEntry}
   > n2 = Is Valid(InputObject=@n1.Item)
   n3 = Set bShouldShowRarity(execute=@n2.Is Not Valid, bShouldShowRarity:bool=true) -> (Output_Get:bool)
   n4 = IsValid(Object=Get ItemDefinition() -> (ItemDefinition:Item Definition Base Object Reference))
   n5 = Get bShouldShowRarity() -> (bShouldShowRarity:bool)
   n6 = Branch(execute=[@n13.then, @n3.then], Condition=OR Boolean(A=NOT Boolean(A=@n4), B=NOT Boolean(A=@n5)))
   > SetVisibility(InVisibility=Collapsed)
   n7 = IsValid(Object=GetSeries!(ItemDefinition:Item Definition Base Object Reference=Get ItemDefinition()) -> (ReturnValue:Fort Item Series Definition Object Reference))
   n8 = Branch(execute=@n6.else, Condition=AND Boolean(A=Get OnlyShowSeries() -> (OnlyShowSeries:bool), B=NOT Boolean(A=@n7)))
   > SetVisibility(InVisibility=Collapsed)
   SetVisibility(execute=@n8.else, InVisibility=SelfHitTestInvisible)
   > n9 = GetRarityData!(
    execute
    Rarity:EFortRarity Enum=GetRarity!(ItemDefinition:Item Definition Base Object Reference=Get ItemDefinition()) -> (ReturnValue:EFortRarity Enum=Common)
   ) -> (then, ReturnValue:Fort Rarity Item Data Structure)
   > "SetText (Text)"(
    self=Get RarityTextBlock() -> (RarityTextBlock:Common Text Object Reference)
    InText=Break Fort Rarity Item Data!(FortRarityItemData:"Fort Rarity Item Data Structure (by ref)"=@n9) -> (
     Name:Text
     Color1:Linear Color Structure
     Color2:Linear Color Structure
     Color3:Linear Color Structure
     Color4:Linear Color Structure
     Color5:Linear Color Structure
     Radius:float=~
     Falloff:float=~
     Brightness:float=~
     Roughness:float=~
     Glow:float=~
    ) {is=title K2Node_BreakStruct}
   )
   n10 = GetSeries(ItemDefinition=Get ItemDefinition())
   > SetActiveWidget(
    self:Widget Switcher Object Reference=Get RarityOrSeriesSwitcher() -> (RarityOrSeriesSwitcher:Widget Switcher Object Reference)
    Widget=Select(
     Option 0:Widget Object Reference=Get RarityTextBlock()
     Option 1:Widget Object Reference=Get SeriesBorder() -> (SeriesBorder:Border Object Reference)
     Index=IsValid(Object=@n10)
    ) -> (ReturnValue:Widget Object Reference)
   )
   n11 = GetSeries(ItemDefinition=Get ItemDefinition())
   > "SetText (Text)"(
    self=Get SeriesTextBlock() -> (SeriesTextBlock:Common Text Object Reference)
    InText=Select(
     Option 0:Text
     Option 1:Text=Get DisplayName!(self:Fort Item Series Definition Object Reference=@n11) -> (DisplayName:Text)
     Index=IsValid(Object=@n11)
    ) -> (ReturnValue:Text)
   )
   n12 = Break Fort Color Palette!(FortColorPalette:"Fort Color Palette Structure (by ref)"=GetRarityOrSeriesColors!(ItemDefinition:Item Definition Base Object Reference=Get ItemDefinition()) -> (ReturnValue:Fort Color Palette Structure)) -> (
    Color1:Linear Color Structure
    Color2:Linear Color Structure
    Color3:Linear Color Structure
    Color4:Linear Color Structure
    Color5:Linear Color Structure
   ) {is=title K2Node_BreakStruct}
   > SetColorAndOpacity(self=Get SeriesTextBlock(), InColorAndOpacity=Make Slate Color(SpecifiedColor=@n12))
   > SetColorAndOpacity(self=Get RarityTextBlock(), InColorAndOpacity=Make Slate Color(SpecifiedColor=@n12))
   n13 = Set bShouldShowRarity(
    execute=@n2.Is Valid
    bShouldShowRarity=Get bShouldShowRarity!(self:Fort Item Details VM Object Reference=Get ItemDetails!(self:Fort Item VM Object Reference=@n1.Item) -> (ItemDetails:Fort Item Details VM Object Reference)) -> (bShouldShowRarity)
   )
   SetBrushColor(self=Get SeriesBorder())
  ]}
]
variables [OnlyShowSeries {type=boolean}]
widgets {
 info {inheritedWidgetCount=0, parentClass=&UserWidget, rootWidgetClass=&WidgetSwitcher, widgetCount=4}
 widgets [
  RarityOrSeriesSwitcher {widgetClassPath=&WidgetSwitcher, bIsVariable=true, widget=RarityOrSeriesSwitcher}
  RarityTextBlock {
   widgetClassPath=&CommonTextBlock
   bIsVariable=true
   parent=RarityOrSeriesSwitcher
   widget=RarityTextBlock
   properties {
    clipping=OnDemand
    justification=Right
    mobileFontSizeMultiplier=1.2
    scrollStyle=&TextScrollStyle-Base_C
    style=/FortUILibrary/Styles/Typography/Foundation/TextStyle-Base-XS-B.TextStyle-Base-XS-B_C
    text="{RARITYNAME}"
    textTransformPolicy=ToUpper
    font {
     bForceMonospaced=false
     bMaterialIsStencil=false
     fontMaterial=null
     fontObject=&NotoSans
     letterSpacing=0
     monospacedWidth=1
     size=16
     skewAmount=0
     typefaceFontName=NotoSans-Bold
     outlineSettings {
      bApplyOutlineToDropShadows=false
      bMiteredCorners=false
      bSeparateFillAlpha=false
      outlineMaterial=null
      outlineSize=0
      outlineColor=<rgba 0 0 0 1>
     }
    }
    shadowOffset=<0 0>
   }
   slot {class=WidgetSwitcherSlot}
  }
  SeriesBorder {
   widgetClassPath=&Border
   bIsVariable=true
   parent=RarityOrSeriesSwitcher
   widget=SeriesBorder
   properties {
    verticalAlignment=VAlign_Center
    background {
     bIsDynamicallyLoaded=false
     drawAs=Box
     imageType=NoImage
     mirroring=NoMirror
     resourceName=None
     resourceObject=/Game/UI/Foundation/Items/Materials/MI_Item_Series
     tiling=NoTile
     imageSize=<32 32>
     margin=<ltrb 0.5 0 0.5 0>
     outlineSettings {
      bUseBrushTransparency=false
      roundingType=HalfHeightRadius
      width=0
      color {colorUseRule=UseColor_Specified, specifiedColor=<rgba 0 0 0 0>}
      cornerRadii=<0 0 0 0>
     }
     tintColor {colorUseRule=UseColor_Specified, specifiedColor=<rgba 1 1 1 0>}
     uVRegion {bIsValid=false, max=<0 0>, min=<0 0>}
    }
    brushColor=<rgba 0.1302 0.1302 0.1302 1>
    padding=<ltrb 0 2 0 2>
   }
   slot {class=WidgetSwitcherSlot, properties {padding=<ltrb 0 -2 0 -2>}}
  }
  SeriesTextBlock {
   widgetClassPath=&CommonTextBlock
   bIsVariable=true
   parent=SeriesBorder
   widget=SeriesTextBlock
   properties {
    clipping=OnDemand
    justification=Right
    scrollStyle=&TextScrollStyle-Base_C
    style=&TextStyle-Base-XS-B_C
    text="{SERIESNAME}"
    textTransformPolicy=ToUpper
    font {
     bForceMonospaced=false
     bMaterialIsStencil=false
     fontMaterial=null
     fontObject=&NotoSans
     letterSpacing=0
     monospacedWidth=1
     size=16
     skewAmount=0
     typefaceFontName=NotoSans-Bold
     outlineSettings {
      bApplyOutlineToDropShadows=false
      bMiteredCorners=false
      bSeparateFillAlpha=false
      outlineMaterial=null
      outlineSize=0
      outlineColor=<rgba 0 0 0 1>
     }
    }
    shadowOffset=<0 0>
   }
   slot {class=BorderSlot, properties {verticalAlignment=VAlign_Center, padding=<ltrb 0 2 0 2>}}
  }
 ]
}
