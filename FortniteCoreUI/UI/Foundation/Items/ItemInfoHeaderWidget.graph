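# Item info header widget (FortItemBaseWidget) showing an item's cosmetic card, name, series or rarity, type, scrollable description, lock state and remaining quantity.
# An item definition change runs Update, which formats the name by quantity and drives rarity or series colors into dynamic font, glow and border materials.
# Setup and setters toggle sections; OverrideHeaderAndDescription supplies custom text and an alternate background.
# UpdateRichDescSize stacks zero-opacity description copies for sizing.
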WidgetBlueprint /FortniteCoreUI/UI/Foundation/Items/ItemInfoHeaderWidget
graphs [
 EventGraph {nodes [
   Event OnItemDefinitionChanged() {is=title K2Node_Event}
   n1 = GetItemDefinition!(self:Fort Item Base Widget Object Reference) -> (ReturnValue:Item Definition Base Object Reference)
   > n2 = Is Valid(InputObject=@n1)
   UpdateCard(execute=@n2.Is Not Valid)
   > UpdateLocked()
   > SetVisibility(
    self=Get Card() -> (Card:Fort Cosmetic Item Card Object Reference)
    InVisibility=Select(Option 0=Collapsed, Option 1=Hidden, Index=Get ShowCard() -> (ShowCard:bool))
   )
   > SetVisibility(
    self=Get RemainingContainer() -> (RemainingContainer:Size Box Object Reference)
    InVisibility=Select(Option 0=Collapsed, Option 1=Hidden, Index=Get ShowRemainingQuantity() -> (ShowRemainingQuantity:bool))
   )
   > SetVisibility(
    self=[
     Get HeaderBGContainer() -> (HeaderBGContainer:Vertical Box Object Reference)
     Get TextContainer() -> (TextContainer:Vertical Box Object Reference)
    ]
    InVisibility=Collapsed
   )
   Update!(
    execute=@n2.Is Valid
    self
    ItemDefinition:Item Definition Base Object Reference=@n1
    Quantity:int=GetQuantity!(self:Fort Item Base Widget Object Reference) -> (ReturnValue:int=0)
    InItem:Fort Item VM Object Reference=Get Item() -> (Item:Fort Item VM Object Reference)
   ) -> (then)
  ]}
 GetAlternateNameForItem {nodes [
   n1 = GetAlternateNameForItem!() -> (then, ItemDef:Item Definition Base Object Reference) {is=title K2Node_FunctionEntry}
   > n2 = Branch(Condition="Equal (Enum)"(
     A:EFortItemType Enum=GetItemType!(ItemDefinition:Item Definition Base Object Reference=@n1) -> (ReturnValue:EFortItemType Enum=WorldItem)
     B:EFortItemType Enum=Worker
    ))
   n3 = Cast To FortWorkerType!(Object=Get ItemDef() -> (ItemDef:Item Definition Base Object Reference)) -> (AsFort Worker Type:Fort Worker Type Object Reference, bSuccess:bool=~) {is=title K2Node_DynamicCast}
   > Return Node!(
    execute
    ReturnValue:Text=Select(
     Option 0:Text="LOCTEXT(\"Survivor\")"
     Option 1:Text="LOCTEXT(\"Lead Survivor\")"
     Index=Get bIsManager!(self:Fort Worker Type Object Reference=@n3) -> (bIsManager:bool)
    ) -> (ReturnValue:Text)
   )
   n4 = Find(
    TargetMap:Map of EFortItemType Enums to Texts=Get ItemAlternateName() -> (ItemAlternateName:Map of EFortItemType Enums to Texts=WorldItem)
    Key:"EFortItemType Enum (by ref)"=GetItemType(ItemDefinition=Get ItemDef()) -> (ReturnValue=WorldItem)
   ) -> (Value:Text)
   n5 = Branch(execute=@n2.else, Condition=@n4.ReturnValue)
   > Return Node(ReturnValue=@n4)
   Return Node(execute=@n5.else, ReturnValue=GetEmptyText!() -> (ReturnValue:Text))
  ]}
 GetTypeText {nodes [
   n1 = GetTypeText!() -> (then, ItemDefinition:Item Definition Base Object Reference, InItem:Fort Item VM Object Reference) {is=title K2Node_FunctionEntry}
   > n2 = Is Valid(InputObject=@n1.InItem)
   n3 = Get ItemDetails!(self:Fort Item VM Object Reference=Get InItem() -> (InItem:Fort Item VM Object Reference)) -> (ItemDetails:Fort Item Details VM Object Reference)
   n4 = GetItemStyleText!(self:Fort Item Details VM Object Reference=@n3) -> (ReturnValue:Text)
   n5 = GetItemCategoryName!(self:Fort Item Details VM Object Reference=@n3) -> (ReturnValue:Text)
   > Return Node!(
    execute
    TypeText:Text=SelectText(
     A=@n5
     B=Format Text!(Format="LOCTEXT(\"{Category} - {Style}\")", Category:Text=@n5, Style:Text=@n4) -> (Result)
     bPickA=TextIsEmpty(InText=@n4)
    )
   )
   n6 = Get ItemDefinition() -> (ItemDefinition:Item Definition Base Object Reference)
   n7 = Get ItemDefinition()
   n8 = GetItemType(ItemDefinition=Get ItemDefinition()) -> (ReturnValue=WorldItem)
   Return Node(
    execute=@n2.Is Not Valid
    TypeText=SelectText(
     A=GetItemTypeName!(ItemDefinition:Item Definition Base Object Reference=@n6, bUsePlural:bool=false) -> (ReturnValue:Text)
     B=GetItemShortDescription!(self:Item Definition Base Object Reference=@n7) -> (ReturnValue:Text)
     bPickA="Equal (Enum)"(A=@n8, B=Currency)
    )
   )
  ]}
 HasValidItemDefinition {nodes [
   HasValidItemDefinition() {is=title K2Node_FunctionEntry}
   > Return Node!(execute, HasValidDescription:bool=IsValid(Object=GetItemDefinition()))
  ]}
 OverrideHeaderAndDescription {nodes [
   n1 = OverrideHeaderAndDescription!() -> (then, Header:Text, Description:Text, ItemType:Text, bUseAlternateHeaderBG:bool=~, Rarity:EFortRarity Enum=Common) {is=title K2Node_FunctionEntry}
   > "SetText (Text)"(
    self=Get NameTextBlock() -> (NameTextBlock:Common Text Object Reference)
    InText=TextToUpper!(InText:"Text (by ref)"=@n1) -> (ReturnValue:Text)
   )
   > "SetText (Text)"(self=Get TypeTextBlock() -> (TypeTextBlock:Common Text Object Reference), InText=@n1.ItemType)
   > SetText(
    self:Item Info Header Description Object Reference=Get RichDescription() -> (RichDescription:Item Info Header Description Object Reference)
    InText:Text=@n1.Description
   )
   > UpdateItemDefinition!(
    execute
    self:Item Series Or Rarity Object Reference=Get ItemSeriesOrRarity() -> (ItemSeriesOrRarity:Item Series Or Rarity Object Reference)
    ItemDefinition:Item Definition Base Object Reference
    Item:Fort Item VM Object Reference
   ) -> (then)
   > SetActiveWidget(
    self:Widget Switcher Object Reference=Get BackgroundSwitcher() -> (BackgroundSwitcher:Widget Switcher Object Reference)
    Widget=Select(
     Option 0:Widget Object Reference=Get HeaderBGContainer()
     Option 1:Widget Object Reference=Get AlternateBGContainer() -> (AlternateBGContainer:Vertical Box Object Reference)
     Index=Get bUseAlternateHeaderBG() -> (bUseAlternateHeaderBG:bool)
    ) -> (ReturnValue:Widget Object Reference)
   )
   > SetVisibility(
    self=Get RarityGlow() -> (RarityGlow:Image Object Reference)
    InVisibility=Select(Option 0:ESlateVisibility Enum=Visible, Option 1:ESlateVisibility Enum=Collapsed, Index=Get bUseAlternateHeaderBG()) -> (ReturnValue:ESlateVisibility Enum)
   )
   > n2 = GetRarityColorPalette!(execute, Rarity:EFortRarity Enum=Get Rarity() -> (Rarity:EFortRarity Enum=Common)) -> (then, ReturnValue:Fort Color Palette Structure)
   > UpdateColorRarity!(execute, self, Colors:"Fort Color Palette Structure (by ref)"=@n2) -> (then)
   > SetVisibility(self=Get RarityTypeSeperatorImage() -> (RarityTypeSeperatorImage:Image Object Reference), InVisibility=Collapsed)
   > UpdateDetailVisibility()
  ]}
 SetLocked {nodes [
   n1 = SetLocked!() -> (then, ShouldShowLocked:bool=~, LockReason:Text) {is=title K2Node_FunctionEntry}
   > Set Locked(Locked:bool=@n1) -> (Output_Get:bool)
   > Set LockedReason(LockedReason:Text=@n1.LockReason) -> (Output_Get:Text)
   > UpdateLocked()
  ]}
 SetRemainingQuantity {nodes [
   n1 = SetRemainingQuantity!() -> (then, ShouldShowRemainingQuantity:bool=~, NewRemainingQuantity:int=-1) {is=title K2Node_FunctionEntry}
   > Set ShowRemainingQuantity(ShowRemainingQuantity:bool=@n1) -> (Output_Get:bool)
   > Set RemainingQuantity(RemainingQuantity:int=@n1.NewRemainingQuantity) -> (Output_Get:int=0)
   > UpdateRemainingQuantity()
  ]}
 SetScrollWidget {nodes [
   SetScrollWidget() {is=title K2Node_FunctionEntry}
   > SetScrollWidget!(execute, ToScroll:Widget Object Reference=Get DetailsScrollBox() -> (DetailsScrollBox:Scroll Box Object Reference)) -> (then) {is=title K2Node_CallFunction}
  ]}
 Setup {nodes [
   n1 = Setup!() -> (then, ShouldShowCard:bool=~, CardSize:EFortItemCardSize Enum=XXS, ShouldShowDetails:bool=~, ShouldShowDescription:bool=~) {is=title K2Node_FunctionEntry}
   > Set ShowCard(ShowCard:bool=@n1) -> (Output_Get:bool)
   > Set CardSize(CardSize:EFortItemCardSize Enum=@n1.CardSize) -> (Output_Get:EFortItemCardSize Enum=XXS)
   > Set ShowDetails(ShowDetails:bool=@n1.ShouldShowDetails) -> (Output_Get:bool)
   > Set ShowDescription(ShowDescription:bool=@n1.ShouldShowDescription) -> (Output_Get:bool)
   > UpdateCard()
   > UpdateDetailVisibility()
  ]}
 Update {nodes [
   n1 = Update!() -> (then, ItemDefinition:Item Definition Base Object Reference, Quantity:int, InItem:Fort Item VM Object Reference) {is=title K2Node_FunctionEntry}
   > UpdateNameText!(execute, self, ItemDefinition:Item Definition Base Object Reference=@n1, Quantity:int=@n1.Quantity) -> (then)
   > n2 = Get!(execute:Exec) -> (then:Exec, else:Exec, InItem:Fort Item VM Object Reference)
   > n3 = GetRarityColorPalette(Rarity=Common)
   n4 = Get ItemDefinition()
   n5 = Get ItemDetails(self=@n2)
   > n6 = UpdateColorRarity(Colors=Select(
     Option 0:"Fort Color Palette Structure (by ref)"=@n3
     Option 1:"Fort Color Palette Structure (by ref)"=GetRarityOrSeriesColors!(ItemDefinition:Item Definition Base Object Reference=@n4) -> (ReturnValue:Fort Color Palette Structure)
     Index=Get bShouldShowRarity!(self:Fort Item Details VM Object Reference=@n5) -> (bShouldShowRarity:bool)
    ) -> (ReturnValue:"Fort Color Palette Structure (by ref)"))
   UpdateItemDefinition(execute=[@n6.then, @n9.then], self=Get ItemSeriesOrRarity(), ItemDefinition=Get ItemDefinition(), Item=Get InItem())
   > "SetText (Text)"(
    self=Get TypeTextBlock()
    InText=GetTypeText!(ItemDefinition:Item Definition Base Object Reference=Get ItemDefinition(), InItem:Fort Item VM Object Reference=Get InItem()) -> (TypeText:Text) {is=title K2Node_CallFunction}
   )
   n7 = NOT Boolean(A=IsVisible(self=Get ItemSeriesOrRarity()))
   n8 = TextIsEmpty(InText=Get Text(self:Text Object Reference=Get TypeTextBlock()))
   > SetVisibility(
    self=Get RarityTypeSeperatorImage()
    InVisibility=Select(Option 0:ESlateVisibility Enum=SelfHitTestInvisible, Option 1:ESlateVisibility Enum=Collapsed, Index=OR Boolean(A=@n8, B=@n7)) -> (ReturnValue:ESlateVisibility Enum)
   )
   > SetText(self=Get RichDescription(), InText=GetRichDescription!(self:Fort Item Base Widget Object Reference) -> (ReturnValue:Text))
   > UpdateCard()
   > UpdateDetailVisibility()
   > UpdateRemainingQuantity()
   > UpdateLocked()
   n9 = UpdateColorRarity(execute=@n2.else, Colors=GetRarityOrSeriesColors(ItemDefinition=Get ItemDefinition()))
  ]}
 UpdateCard {nodes [
   UpdateCard() {is=title K2Node_FunctionEntry}
   > n1 = Branch(Condition=Get ShowCard())
   n2 = Get CardContainer() -> (CardContainer:Vertical Box Object Reference)
   n3 = Get Card()
   > SetVisibility(self=[@n2, @n3], InVisibility=HitTestInvisible)
   n4 = GetItemDefinition()
   > n5 = Is Valid(InputObject=@n4)
   > SetItemToRepresent!(
    execute
    self:Fort Cosmetic Item Card Object Reference=Get Card()
    InItemToRepresent:Fort Item Object Reference=CreateTemporaryItemInstance!(ItemDefinition:Item Definition Base Object Reference=@n4, Count:int=GetQuantity() -> (ReturnValue=0), Level:int=-1) -> (ReturnValue:Fort Item Object Reference)
    InItemVM:Fort Item VM Object Reference=Get Item()
   ) -> (then)
   SetItemToRepresent(execute=@n5.Is Not Valid, self=Get Card())
   SetVisibility(execute=@n1.else, self=[@n2, @n3], InVisibility=Collapsed)
  ]}
 UpdateColorRarity {nodes [
   n1 = UpdateColorRarity!() -> (then, Colors:"Fort Color Palette Structure (by ref)") {is=title K2Node_FunctionEntry}
   > n2 = Sequence!(execute) -> (then_0, then_1, then_2:Exec)
   n3 = Get NameTextBlock()
   > SetStyle!(
    execute
    self:Common Text Object Reference=@n3
    InStyle:Common Text Style Class Reference=/FortUILibrary/Styles/Typography/Foundation/TextStyle-Header-M-SimpleGradient.TextStyle-Header-M-SimpleGradient_C
   ) -> (then)
   > n4 = GetDynamicFontMaterial!(execute, self:Text Object Reference=@n3) -> (then, ReturnValue:Material Instance Dynamic Object Reference)
   n5 = Break Fort Color Palette!(FortColorPalette:"Fort Color Palette Structure (by ref)"=@n1) -> (
    Color1:Linear Color Structure
    Color2:Linear Color Structure
    Color3:Linear Color Structure
    Color4:Linear Color Structure
    Color5:Linear Color Structure
   ) {is=title K2Node_BreakStruct}
   > SetVectorParameterValue(self=@n4, ParameterName=RGB1, Value=@n5)
   n6 = GetDynamicMaterial(execute=@n2.then_1, self=Get RarityGlow())
   > SetVectorParameterValue(self=@n6, ParameterName=vRGB, Value=@n5)
   n7 = GetDynamicMaterial(execute=@n2.then_2, self:Border Object Reference=Get HeaderBackground() -> (HeaderBackground:Common Border Object Reference))
   > SetVectorParameterValue(self=@n7, ParameterName=RGBA0, Value=@n5.Color3)
   > SetVectorParameterValue(self=@n7, ParameterName=RGBA1, Value=@n5.Color2)
   > n8 = GetDynamicMaterial(self=Get HeaderRule() -> (HeaderRule:Common Border Object Reference))
   > SetVectorParameterValue(self=@n8, ParameterName=RGBA0, Value=@n5.Color3)
   > SetVectorParameterValue(self=@n8, ParameterName=RGBA1, Value=@n5.Color2)
  ]}
 UpdateDetailVisibility {nodes [
   UpdateDetailVisibility() {is=title K2Node_FunctionEntry}
   > n1 = Cast To OverlaySlot!(
    execute
    Object=Get Slot!(self:Widget Object Reference=Get MainHB() -> (MainHB:Horizontal Box Object Reference)) -> (Slot:Panel Slot Object Reference)
   ) -> (then, CastFailed, AsOverlay Slot:Overlay Slot Object Reference) {is=title K2Node_DynamicCast}
   n2 = Get ShowDetails() -> (ShowDetails:bool)
   > n3 = SetHorizontalAlignment(
    self=@n1
    InHorizontalAlignment=Select(Option 0:EHorizontalAlignment Enum=HAlign_Center, Option 1:EHorizontalAlignment Enum=HAlign_Fill, Index=@n2) -> (ReturnValue:EHorizontalAlignment Enum)
   )
   n4 = Branch(execute=[@n3.then, @n1.CastFailed], Condition=@n2)
   n5 = Get HeaderBGContainer()
   n6 = Get TextContainer()
   > SetVisibility(self=[@n5, @n6], InVisibility=SelfHitTestInvisible)
   > n7 = Branch(Condition=Get ShowDescription() -> (ShowDescription:bool))
   n8 = Get DetailsScrollBox()
   > SetVisibility(self=@n8, InVisibility=Visible)
   SetVisibility(execute=@n7.else, self=@n8, InVisibility=Collapsed)
   SetVisibility(execute=@n4.else, self=[@n5, @n6], InVisibility=Collapsed)
  ]}
 UpdateLocked {nodes [
   UpdateLocked() {is=title K2Node_FunctionEntry}
   > n1 = Branch(Condition=Get Locked() -> (Locked:bool))
   n2 = Get LockImage() -> (LockImage:Image Object Reference)
   n3 = Get LockedTextBlock() -> (LockedTextBlock:Common Text Object Reference)
   > SetVisibility(self=[@n2, @n3], InVisibility=SelfHitTestInvisible)
   > "SetText (Text)"(self=@n3, InText=Get LockedReason() -> (LockedReason:Text))
   SetVisibility(execute=@n1.else, self=[@n2, @n3], InVisibility=Collapsed)
  ]}
 UpdateNameText {nodes [
   UpdateNameText!() -> (then, ItemDefinition:Item Definition Base Object Reference, Quantity:int) {is=title K2Node_FunctionEntry}
   > n1 = Set DisplayName(DisplayName:Text=GetDisplayName!(self:Item Definition Base Object Reference=Get ItemDefinition()) -> (ReturnValue:Text)) -> (Output_Get:Text)
   > n2 = Branch(Condition=TextIsEmpty(InText=@n1))
   > "SetText (Text)"(
    self=Get NameTextBlock()
    InText=GetAlternateNameForItem!(ItemDef:Item Definition Base Object Reference=Get ItemDefinition()) -> (ReturnValue:Text) {is=title K2Node_CallFunction}
   )
   n3 = Branch(execute=@n2.else, Condition="integer > integer"(A=Get Quantity() -> (Quantity:int=0), B=1) {is=title K2Node_CallFunction})
   n4 = Get ItemDefinition()
   > n5 = Cast To FortTokenType!(execute, Object=@n4) -> (then, CastFailed, AsFort Token Type:Fort Token Type Object Reference) {is=title K2Node_DynamicCast}
   > n6 = Branch(Condition=Get bPercentageRepresentation!(self:Fort Token Type Object Reference=@n5) -> (bPercentageRepresentation:bool))
   n7 = Get DisplayName() -> (DisplayName:Text)
   > "SetText (Text)"(
    self=Get NameTextBlock()
    InText=Format Text!(Format="LOCTEXT(\"{Percent}% {BonusName}\")", Percent:int=Get Quantity() -> (Quantity=0), BonusName:Text=TextToUpper(InText=@n7)) -> (Result)
   )
   n8 = GetTemplateId!(self:Mcp Item Definition Base Object Reference=@n4) -> (ReturnValue:String)
   n9 = TextToUpper(InText=Get DisplayName())
   n10 = Get Quantity() -> (Quantity=0)
   "SetText (Text)"(
    execute=[@n6.else, @n5.CastFailed]
    self=Get NameTextBlock()
    InText=Select(
     Option 0:Text=Format Text!(Format="LOCTEXT(\"{Quantity}x {ItemName}\")", Quantity:int=@n10, ItemName:Text=@n9) -> (Result)
     Option 1:Text=Format Text(Format="LOCTEXT(\"{Quantity} {ItemName}\")", Quantity=@n10, ItemName=@n9)
     Index="Equal Exactly (String)"(A=@n8, B=Token:athenabattlepasstier)
    ) -> (ReturnValue:Text)
   )
   "SetText (Text)"(execute=@n3.else, self=Get NameTextBlock(), InText=TextToUpper(InText=Get DisplayName()))
  ]}
 UpdateRemainingQuantity {nodes [
   UpdateRemainingQuantity() {is=title K2Node_FunctionEntry}
   > n1 = Branch(Condition=Get ShowRemainingQuantity())
   n2 = Get RemainingContainer()
   > SetVisibility(self=@n2, InVisibility=SelfHitTestInvisible)
   n3 = Get RemainingQuantity() -> (RemainingQuantity:int=0)
   n4 = "To Text (Integer)"(Value=@n3)
   > "SetText (Text)"(
    self=Get RemainingText() -> (RemainingText:Common Text Object Reference)
    InText=Select(
     Option 0=Format Text!(Format="LOCTEXT(\"{Quantity} left\")", Quantity:Text=@n4) -> (Result)
     Option 1="LOCTEXT(\"Unlimited\")"
     Index="integer < integer"!(A:int=@n3, B:int=0) -> (ReturnValue:bool) {is=title K2Node_CallFunction}
    )
   )
   SetVisibility(execute=@n1.else, self=@n2, InVisibility=Collapsed)
  ]}
 UpdateRichDescSize {nodes [
   n1 = UpdateRichDescSize!() -> (then, RichDescriptions:Array of Texts) {is=title K2Node_FunctionEntry}
   > ClearChildren(self=Get Overlay_RichDescriptionsForSizing() -> (Overlay_RichDescriptionsForSizing:Overlay Object Reference))
   > n2 = For Each Loop(Array:Array of Texts=@n1) -> (Array Element:Text)
   > n3 = Create Widget!(
    execute:Exec
    Class:User Widget Class Reference=/Game/UI/Foundation/Items/ItemInfoHeaderDescription.ItemInfoHeaderDescription_C
    OwningPlayer:Player Controller Object Reference=GetOwningPlayer()
   ) -> (then:Exec, ReturnValue:Item Info Header Description Object Reference)
   > n4 = AddChildToOverlay!(execute, self:Overlay Object Reference=Get Overlay_RichDescriptionsForSizing(), Content:Widget Object Reference=@n3) -> (then, ReturnValue:Overlay Slot Object Reference)
   > SetHorizontalAlignment(self=@n4, InHorizontalAlignment=HAlign_Fill)
   > SetVerticalAlignment(self=@n4, InVerticalAlignment=VAlign_Fill)
   > SetText(self=@n3, InText=@n2)
  ]}
]
variables [
 CardSize {type=EFortItemCardSize, default=L}
 ItemAlternateName {default {homebaseBannerIcon=BANNER ICON}}
 Locked {type=boolean}
 LockedReason {type=string, default=""}
 RemainingQuantity {type=integer, default=-1}
 ShowCard {type=boolean, default=true}
 ShowDescription {type=boolean, default=true}
 ShowDetails {type=boolean, default=true}
 ShowRemainingQuantity {type=boolean}
]
widgets {
 info {inheritedWidgetCount=0, parentClass=/Script/FortniteUI.FortItemBaseWidget, rootWidgetClass=&CommonBorder, widgetCount=38}
 widgets [
  RootBackground {
   widgetClassPath=&CommonBorder
   bIsVariable=true
   widget=RootBackground
   properties {style=/FortUILibrary/Styles/Border/Border_Solid_MdBlue.Border_Solid_MdBlue_C, brushColor=<rgba 1 1 1 0.5>, padding=<ltrb 0 0 0 0>}
  }
  Overlay_4 {widgetClassPath=&Overlay, parent=RootBackground, widget=Overlay_4, slot {class=BorderSlot, properties {padding=<ltrb 0 0 0 0>}}}
  BackgroundSwitcher {
   widgetClassPath=&WidgetSwitcher
   bIsVariable=true
   parent=Overlay_4
   widget=BackgroundSwitcher
   slot {class=OverlaySlot, properties {horizontalAlignment=HAlign_Fill}}
  }
  HeaderBGContainer {
   widgetClassPath=&VerticalBox
   bIsVariable=true
   parent=BackgroundSwitcher
   widget=HeaderBGContainer
   slot {class=WidgetSwitcherSlot}
  }
  HeaderBackground {
   widgetClassPath=&CommonBorder
   bIsVariable=true
   parent=HeaderBGContainer
   widget=HeaderBackground
   properties {style=/FortUILibrary/Styles/Border/Border-AthenaItemInfoHeader.Border-AthenaItemInfoHeader_C, padding=<ltrb 0 0 0 0>}
   slot {class=VerticalBoxSlot, properties {size {sizeRule=Fill, value=1}}}
  }
  Spacer_0 {
   widgetClassPath=&Spacer
   parent=HeaderBackground
   widget=Spacer_0
   properties {size=<1 103>}
   slot {class=BorderSlot, properties {padding=<ltrb 0 0 0 0>}}
  }
  HeaderRule {
   widgetClassPath=&CommonBorder
   bIsVariable=true
   parent=HeaderBGContainer
   widget=HeaderRule
   properties {
    style=/FortUILibrary/Styles/Border/Border-AthenaItemInfoRule.Border-AthenaItemInfoRule_C
    background {
     bIsDynamicallyLoaded=false
     drawAs=Box
     imageType=NoImage
     mirroring=NoMirror
     resourceName=None
     resourceObject=/FortUILibrary/Styles/Border/Materials/M-Border-ItemInfoHeader_procedural
     tiling=NoTile
     imageSize=<128 32>
     margin=<ltrb 0.1 0.7965 1 0.2611>
     outlineSettings {
      bUseBrushTransparency=false
      roundingType=HalfHeightRadius
      width=0
      color {colorUseRule=UseColor_Specified, specifiedColor=<rgba 0 0 0 0>}
      cornerRadii=<0 0 0 0>
     }
     tintColor {colorUseRule=UseColor_Specified, specifiedColor=<rgba 1 1 1 1>}
     uVRegion {bIsValid=false, max=<0 0>, min=<0 0>}
    }
    padding=<ltrb 0 0 0 0>
   }
   slot {class=VerticalBoxSlot}
  }
  Spacer_1 {
   widgetClassPath=&Spacer
   parent=HeaderRule
   widget=Spacer_1
   properties {size=<1 4>}
   slot {class=BorderSlot, properties {padding=<ltrb 0 0 0 0>}}
  }
  AlternateBGContainer {
   widgetClassPath=&VerticalBox
   bIsVariable=true
   parent=BackgroundSwitcher
   widget=AlternateBGContainer
   slot {class=WidgetSwitcherSlot, properties {verticalAlignment=VAlign_Top}}
  }
  AlternateBorder {
   widgetClassPath=&CommonBorder
   bIsVariable=true
   parent=AlternateBGContainer
   widget=AlternateBorder
   properties {style=/FortUILibrary/Styles/Border/Border-ItemInfo-Blank.Border-ItemInfo-Blank_C, padding=<ltrb 0 0 0 0>}
   slot {class=VerticalBoxSlot, properties {size {sizeRule=Fill, value=1}}}
  }
  Spacer_2 {
   widgetClassPath=&Spacer
   parent=AlternateBorder
   widget=Spacer_2
   properties {size=<1 96>}
   slot {class=BorderSlot, properties {padding=<ltrb 0 0 0 0>}}
  }
  AlternateRule {
   widgetClassPath=&CommonBorder
   bIsVariable=true
   parent=AlternateBGContainer
   widget=AlternateRule
   properties {
    style=&Border-ItemInfo-Blank_C
    background {
     bIsDynamicallyLoaded=false
     drawAs=Box
     imageType=NoImage
     mirroring=NoMirror
     resourceName=None
     resourceObject=&M-Border-ItemInfoHeader_procedural
     tiling=NoTile
     imageSize=<128 32>
     margin=<ltrb 0.1 0.7965 1 0.2611>
     outlineSettings {
      bUseBrushTransparency=false
      roundingType=HalfHeightRadius
      width=0
      color {colorUseRule=UseColor_Specified, specifiedColor=<rgba 0 0 0 0>}
      cornerRadii=<0 0 0 0>
     }
     tintColor {colorUseRule=UseColor_Specified, specifiedColor=<rgba 1 1 1 1>}
     uVRegion {bIsValid=false, max=<0 0>, min=<0 0>}
    }
    padding=<ltrb 0 0 0 0>
   }
   slot {class=VerticalBoxSlot}
  }
  Spacer_3 {
   widgetClassPath=&Spacer
   parent=AlternateRule
   widget=Spacer_3
   properties {size=<1 4>}
   slot {class=BorderSlot, properties {padding=<ltrb 0 0 0 0>}}
  }
  MainHB {
   widgetClassPath=&HorizontalBox
   bIsVariable=true
   parent=Overlay_4
   widget=MainHB
   slot {class=OverlaySlot, properties {horizontalAlignment=HAlign_Fill, verticalAlignment=VAlign_Fill, padding=<ltrb 8 8 8 0>}}
  }
  CardContainer {
   widgetClassPath=&VerticalBox
   bIsVariable=true
   parent=MainHB
   widget=CardContainer
   slot {class=HorizontalBoxSlot, properties {horizontalAlignment=HAlign_Left, verticalAlignment=VAlign_Top, padding=<ltrb 8 8 8 16>}}
  }
  Overlay_5 {widgetClassPath=&Overlay, parent=CardContainer, widget=Overlay_5, slot {class=VerticalBoxSlot}}
  SoftShadow {
   widgetClassPath=&Image
   parent=Overlay_5
   widget=SoftShadow
   properties {
    visibility=SelfHitTestInvisible
    brush {
     bIsDynamicallyLoaded=false
     drawAs=Box
     imageType=NoImage
     mirroring=NoMirror
     resourceName=None
     resourceObject=/FortUILibrary/Styles/Border/Materials/M_Shadow_BoxScaleSoft
     tiling=NoTile
     imageSize=<96 96>
     margin=<ltrb 0.5 0.5 0.5 0.5>
     outlineSettings {
      bUseBrushTransparency=false
      roundingType=HalfHeightRadius
      width=0
      color {colorUseRule=UseColor_Specified, specifiedColor=<rgba 0 0 0 0>}
      cornerRadii=<0 0 0 0>
     }
     tintColor {colorUseRule=UseColor_Specified, specifiedColor=<rgba 1 1 1 1>}
     uVRegion {bIsValid=false, max=<0 0>, min=<0 0>}
    }
    colorAndOpacity=<rgba 1 1 1 0.2>
   }
   slot {class=OverlaySlot, properties {horizontalAlignment=HAlign_Fill, verticalAlignment=VAlign_Fill, padding=<ltrb -24 -24 -24 -24>}}
  }
  ItemCardL {
   widgetClassPath=&Overlay
   parent=Overlay_5
   widget=ItemCardL
   slot {class=OverlaySlot, properties {horizontalAlignment=HAlign_Center, verticalAlignment=VAlign_Fill}}
  }
  Card {
   widgetClassPath=/FortniteCoreUI/Athena/UI/Frontend/CosmeticItemCard/FortCosmeticItemCard
   bIsVariable=true
   parent=ItemCardL
   widget=Card
   properties {cardWidth=160}
   slot {class=OverlaySlot}
  }
  LockImage {
   widgetClassPath=&Image
   bIsVariable=true
   parent=Overlay_5
   widget=LockImage
   properties {
    visibility=SelfHitTestInvisible
    brush {
     bIsDynamicallyLoaded=false
     drawAs=Image
     imageType=NoImage
     mirroring=NoMirror
     resourceName=None
     resourceObject=/Game/UI/Foundation/Textures/Icons/Locks/T-Icon-Lock-32
     tiling=NoTile
     imageSize=<32 32>
     margin=<ltrb 0 0 0 0>
     outlineSettings {
      bUseBrushTransparency=false
      roundingType=HalfHeightRadius
      width=0
      color {colorUseRule=UseColor_Specified, specifiedColor=<rgba 0 0 0 0>}
      cornerRadii=<0 0 0 0>
     }
     tintColor {colorUseRule=UseColor_Specified, specifiedColor=<rgba 1 1 1 1>}
     uVRegion {bIsValid=false, max=<0 0>, min=<0 0>}
    }
   }
   slot {class=OverlaySlot, properties {horizontalAlignment=HAlign_Right, padding=<ltrb 0 8 4 0>}}
  }
  RemainingContainer {
   widgetClassPath=&SizeBox
   bIsVariable=true
   parent=CardContainer
   widget=RemainingContainer
   properties {heightOverride=35}
   slot {class=VerticalBoxSlot, properties {horizontalAlignment=HAlign_Center, verticalAlignment=VAlign_Top, padding=<ltrb 8 8 8 8>}}
  }
  CommonBorder_2 {
   widgetClassPath=&CommonBorder
   parent=RemainingContainer
   widget=CommonBorder_2
   properties {style=/FortUILibrary/Styles/Border/Border-RoundBoxM.Border-RoundBoxM_C, brushColor=<rgba 0.0021 0.0048 0.0343 1>, padding=<ltrb 10 4 10 4>}
   slot {class=SizeBoxSlot}
  }
  RemainingText {
   widgetClassPath=&CommonTextBlock
   bIsVariable=true
   parent=CommonBorder_2
   widget=RemainingText
   properties {
    style=/FortUILibrary/Styles/Typography/Foundation/TextStyle-Base-S-B.TextStyle-Base-S-B_C
    text=x888
    font {
     bForceMonospaced=false
     bMaterialIsStencil=false
     fontMaterial=null
     fontObject=&NotoSans
     letterSpacing=0
     monospacedWidth=1
     size=18
     skewAmount=0
     typefaceFontName=NotoSans-Bold
     outlineSettings {
      bApplyOutlineToDropShadows=false
      bMiteredCorners=false
      bSeparateFillAlpha=false
      outlineMaterial=null
      outlineSize=0
      outlineColor=<rgba 0 0 0 1>
     }
    }
    shadowOffset=<0 0>
   }
   slot {class=BorderSlot, properties {padding=<ltrb 10 4 10 4>}}
  }
  TextContainer {
   widgetClassPath=&VerticalBox
   bIsVariable=true
   parent=MainHB
   widget=TextContainer
   slot {class=HorizontalBoxSlot, properties {padding=<ltrb 8 8 8 0>, size {sizeRule=Fill, value=1}}}
  }
  SizeBox_2 {
   widgetClassPath=&SizeBox
   parent=TextContainer
   widget=SizeBox_2
   properties {bOverride_HeightOverride=true, heightOverride=80}
   slot {class=VerticalBoxSlot, properties {verticalAlignment=VAlign_Top}}
  }
  VerticalBox_1 {widgetClassPath=&VerticalBox, parent=SizeBox_2, widget=VerticalBox_1, slot {class=SizeBoxSlot}}
  HorizontalBox_1 {
   widgetClassPath=&HorizontalBox
   parent=VerticalBox_1
   widget=HorizontalBox_1
   slot {class=VerticalBoxSlot, properties {padding=<ltrb 0 -6 0 -6>}}
  }
  ItemSeriesOrRarity {
   widgetClassPath=/Game/UI/Foundation/Items/ItemSeriesOrRarity
   bIsVariable=true
   parent=HorizontalBox_1
   widget=ItemSeriesOrRarity
   slot {class=HorizontalBoxSlot, properties {horizontalAlignment=HAlign_Left, verticalAlignment=VAlign_Top}}
  }
  RarityTypeSeperatorImage {
   widgetClassPath=&Image
   bIsVariable=true
   parent=HorizontalBox_1
   widget=RarityTypeSeperatorImage
   properties {
    brush {
     bIsDynamicallyLoaded=false
     drawAs=Image
     imageType=NoImage
     mirroring=NoMirror
     resourceName=None
     resourceObject=null
     tiling=NoTile
     imageSize=<3 20>
     margin=<ltrb 0 0 0 0>
     outlineSettings {
      bUseBrushTransparency=false
      roundingType=HalfHeightRadius
      width=0
      color {colorUseRule=UseColor_Specified, specifiedColor=<rgba 0 0 0 0>}
      cornerRadii=<0 0 0 0>
     }
     tintColor {colorUseRule=UseColor_Specified, specifiedColor=<rgba 1 1 1 1>}
     uVRegion {bIsValid=false, max=<0 0>, min=<0 0>}
    }
    colorAndOpacity=<rgba 1 1 1 0.5>
   }
   slot {class=HorizontalBoxSlot, properties {horizontalAlignment=HAlign_Center, verticalAlignment=VAlign_Center, padding=<ltrb 8 0 8 0>}}
  }
  TypeTextBlock {
   widgetClassPath=&CommonTextBlock
   bIsVariable=true
   parent=HorizontalBox_1
   widget=TypeTextBlock
   properties {
    mobileFontSizeMultiplier=1.2
    scrollStyle=&TextScrollStyle-Base_C
    style=/FortUILibrary/Styles/Typography/Foundation/TextStyle-Base-XS-B-70pc.TextStyle-Base-XS-B-70pc_C
    text="{TYPENAME}"
    textTransformPolicy=ToUpper
    colorAndOpacity {colorUseRule=UseColor_Specified, specifiedColor=<rgba 1 1 1 0.7>}
    font {
     bForceMonospaced=false
     bMaterialIsStencil=false
     fontMaterial=null
     fontObject=&NotoSans
     letterSpacing=0
     monospacedWidth=1
     size=16
     skewAmount=0
     typefaceFontName=NotoSans-Bold
     outlineSettings {
      bApplyOutlineToDropShadows=false
      bMiteredCorners=false
      bSeparateFillAlpha=false
      outlineMaterial=null
      outlineSize=0
      outlineColor=<rgba 0 0 0 1>
     }
    }
    shadowOffset=<0 0>
   }
   slot {class=HorizontalBoxSlot, properties {horizontalAlignment=HAlign_Left, verticalAlignment=VAlign_Center, size {sizeRule=Fill, value=1}}}
  }
  Overlay_1 {
   widgetClassPath=&Overlay
   parent=VerticalBox_1
   widget=Overlay_1
   slot {class=VerticalBoxSlot, properties {horizontalAlignment=HAlign_Left, verticalAlignment=VAlign_Top, padding=<ltrb 0 8 0 0>}}
  }
  RarityGlow {
   widgetClassPath=&Image
   bIsVariable=true
   parent=Overlay_1
   widget=RarityGlow
   properties {
    brush {
     bIsDynamicallyLoaded=false
     drawAs=Box
     imageType=NoImage
     mirroring=NoMirror
     resourceName=None
     resourceObject=/FortUILibrary/Styles/Border/Materials/M_Border_SoftGlow
     tiling=NoTile
     imageSize=<128 32>
     margin=<ltrb 0.5 0 0.5 0>
     outlineSettings {
      bUseBrushTransparency=false
      roundingType=HalfHeightRadius
      width=0
      color {colorUseRule=UseColor_Specified, specifiedColor=<rgba 0 0 0 0>}
      cornerRadii=<0 0 0 0>
     }
     tintColor {colorUseRule=UseColor_Specified, specifiedColor=<rgba 1 1 1 1>}
     uVRegion {bIsValid=false, max=<0 0>, min=<0 0>}
    }
    colorAndOpacity=<rgba 1 1 1 0.4>
   }
   slot {class=OverlaySlot, properties {horizontalAlignment=HAlign_Fill, verticalAlignment=VAlign_Fill, padding=<ltrb -32 -16 -32 -16>}}
  }
  NameTextBlock {
   widgetClassPath=&CommonTextBlock
   bIsVariable=true
   parent=Overlay_1
   widget=NameTextBlock
   properties {
    mobileFontSizeMultiplier=1.2
    scrollStyle=&TextScrollStyle-Base_C
    style=&TextStyle-Header-M-SimpleGradient_C
    text="{ITEMNAME}"
    font {
     bForceMonospaced=false
     bMaterialIsStencil=false
     fontMaterial=/FortUILibrary/Materials/Foundation/M_Font_SimpleGradient_Dynamic
     fontObject=&Burbank
     letterSpacing=0
     monospacedWidth=1
     size=32
     skewAmount=0
     typefaceFontName=Burbank-Black
     outlineSettings {
      bApplyOutlineToDropShadows=false
      bMiteredCorners=false
      bSeparateFillAlpha=false
      outlineMaterial=null
      outlineSize=0
      outlineColor=<rgba 0 0 0 1>
     }
    }
    shadowOffset=<0 0>
   }
   slot {class=OverlaySlot, properties {horizontalAlignment=HAlign_Fill, verticalAlignment=VAlign_Center}}
  }
  DetailsScrollBox {
   widgetClassPath=&ScrollBox
   bIsVariable=true
   parent=TextContainer
   widget=DetailsScrollBox
   slot {class=VerticalBoxSlot, properties {padding=<ltrb 0 20 0 16>, size {sizeRule=Fill, value=1}}}
  }
  LockedTextBlock {
   widgetClassPath=&CommonTextBlock
   bIsVariable=true
   parent=DetailsScrollBox
   widget=LockedTextBlock
   properties {
    mobileFontSizeMultiplier=1.2
    style=/FortUILibrary/Styles/Typography/Foundation/TextStyle-Base-XS-70pc.TextStyle-Base-XS-70pc_C
    text=Unlock at Level 5
    colorAndOpacity {colorUseRule=UseColor_Specified, specifiedColor=<rgba 1 1 1 0.7>}
    font {
     bForceMonospaced=false
     bMaterialIsStencil=false
     fontMaterial=null
     fontObject=&NotoSans
     letterSpacing=0
     monospacedWidth=1
     size=16
     skewAmount=0
     typefaceFontName=NotoSans-Regular
     outlineSettings {
      bApplyOutlineToDropShadows=false
      bMiteredCorners=false
      bSeparateFillAlpha=false
      outlineMaterial=null
      outlineSize=0
      outlineColor=<rgba 0 0 0 1>
     }
    }
    shadowOffset=<0 0>
   }
   slot {class=ScrollBoxSlot}
  }
  Overlay_0 {widgetClassPath=&Overlay, parent=DetailsScrollBox, widget=Overlay_0, slot {class=ScrollBoxSlot}}
  RichDescription {
   widgetClassPath=/Game/UI/Foundation/Items/ItemInfoHeaderDescription
   bIsVariable=true
   parent=Overlay_0
   widget=RichDescription
   slot {class=OverlaySlot, properties {horizontalAlignment=HAlign_Fill, verticalAlignment=VAlign_Fill}}
  }
  Overlay_RichDescriptionsForSizing {
   widgetClassPath=&Overlay
   bIsVariable=true
   parent=Overlay_0
   widget=Overlay_RichDescriptionsForSizing
   properties {renderOpacity=0}
   slot {class=OverlaySlot, properties {horizontalAlignment=HAlign_Fill, verticalAlignment=VAlign_Fill}}
  }
 ]
}
