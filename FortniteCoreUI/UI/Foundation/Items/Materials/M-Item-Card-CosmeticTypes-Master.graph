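# Translucent UI master material for cosmetic item cards, showing an icon over a rarity-colored background inside a border.
# Radial gradients and slanted, rotated-UV masks shape the background bands and clip hero icons.
# Scalar switches blend between hero, item and variant layouts, while icon UVs scale down for undersized textures.
# A rounded stroke mask overlays the border with an optional time-animated sweep, and loading or visibility parameters hide the icon.
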Material /FortniteCoreUI/UI/Foundation/Items/Materials/M-Item-Card-CosmeticTypes-Master
graph {
 properties {
  bUsedWithStaticMesh=false
  blendMode=BLEND_Translucent
  materialDomain=MD_UI
  pixelDepthOffsetMode=PDOM_Legacy
  usedShadingModels=Default Lit
 }
 nodes [
  n1 = VectorParameter() {group=Item_Rarity_Colors, parameterName=Rarity 2, defaultValue=<rgba 0.0574 0.0624 0.07 1>}
  n2 = MaterialFunctionCall() {materialFunction=&GetUserInterfaceUV}
  n3 = ComponentMask(None=@n2.Normalized UV) {r=true}
  n4 = LinearInterpolate(A=VectorParameter() {group=Item_Rarity_Colors, parameterName=Rarity 3, defaultValue=<rgba 0.0648 0.0931 0.1356 1>}.RGB, B=@n1.RGB, Alpha=@n3)
  n5 = LinearInterpolate(
   A=VectorParameter() {group=Item_Rarity_Colors, parameterName=Rarity 4, defaultValue=<rgba 0.0296 0.0467 0.0762 1>}.RGB
   B=Multiply(A=@n1.RGB) {constB=0.75}
   Alpha=@n3
  )
  n6 = LinearInterpolate(
   A=@n2.Normalized UV
   B=Multiply(A=@n2.Normalized UV, B=Constant2Vector() {g=0.7, r=1})
   Alpha=ScalarParameter() {group=ItemType, parameterName=bIsItem, sliderMax=1}
  )
  n7 = Constant2Vector() {g=1, r=0.45}
  n8 = Constant() {r=0.5}
  n9 = Constant() {r=0.502}
  n10 = MaterialFunctionCall(
   Alpha=MaterialFunctionCall!(
    UVs=Multiply(A=@n6, B=@n7)
    CenterPosition=Constant2Vector() {g=-1.6, r=-0.85}
    Radius=ScalarParameter() {defaultValue=3, group=Item_Type_CheckerBacking, parameterName=gInner Circle_a}
   ) {materialFunction=/Engine/Functions/Engine_MaterialFunctions01/Gradient/SphereGradient-2D}.Result 0-1
   Min=@n8
   Max=@n9
  ) {materialFunction=&SmoothStep}
  n11 = Constant2Vector() {g=1, r=0.5}
  n12 = MaterialFunctionCall(
   Alpha=MaterialFunctionCall(
    UVs=Multiply(A=@n6, B=@n11)
    CenterPosition=Constant2Vector() {g=2.5, r=1.75}
    Radius=ScalarParameter() {defaultValue=3, group=Item_Type_CheckerBacking, parameterName=gInner Circle_b}
   ) {materialFunction=&SphereGradient-2D}.Result 0-1
   Min=@n8
   Max=@n9
  ) {materialFunction=&SmoothStep}
  n13 = MaterialFunctionCall(
   Alpha=MaterialFunctionCall(
    UVs=@n6
    CenterPosition=Constant2Vector() {g=1.5, r=-1.6}
    Radius=ScalarParameter() {defaultValue=3, group=Item_Type_CheckerBacking, parameterName=gInner Circle_c}
   ) {materialFunction=&SphereGradient-2D}.Result 0-1
   Min=@n8
   Max=@n9
  ) {materialFunction=&SmoothStep}
  n14 = DotProduct(A=@n12.Result, B=@n13.Result)
  n15 = OneMinus(None=@n12.Result)
  n16 = OneMinus(None=@n13.Result)
  n17 = Clamp(None=Add(A=DotProduct(A=@n10.Result, B=@n14), B=DotProduct(A=@n15, B=@n16)))
  n18 = OneMinus()
  Desaturation(None=LinearInterpolate(A=LinearInterpolate(A=@n4, B=@n5, Alpha=@n17), Alpha=Clamp(None=@n18)))
  n19 = Constant2Vector() {r=1}
  n20 = Constant() {r=1.125}
  n21 = Constant() {r=1.3}
  n22 = MaterialFunctionCall() {materialFunction=&GetUserInterfaceUV}
  n23 = VectorParameter() {group=Item_Rarity_Colors, parameterName=Rarity 1, defaultValue=<rgba 0.956 0.9647 0.9734 1>}
  n24 = LinearInterpolate(
   A=VectorParameter() {parameterName=Rarity 3, defaultValue=<rgba 0.0296 0.0467 0.0762 1>}.RGB
   B=@n23.RGB
   Alpha=Clamp(None=MaterialFunctionCall(UVs=@n22.Normalized UV, CenterPosition=@n19, Radius=@n20, Density=@n21) {materialFunction=&RadialGradientExponential}.RadialGradientExponential)
  )
  n25 = Add(
   A=Rotator(Coordinate=@n22.Normalized UV, Time=ScalarParameter() {defaultValue=-8.5, group=Item_Border, parameterName=gRotation})
   B=Multiply(A=Time(), B=ScalarParameter() {defaultValue=0.25, group=Item_General, parameterName=gTime})
  )
  n26 = Abs(None=Add(A=Frac(None=@n25)) {constB=-0.5})
  n27 = Clamp(None=MaterialFunctionCall(
    Alpha=ComponentMask(None=@n26) {r=true}
    Min=ScalarParameter() {group=Item_Border, parameterName=gGradientMin}
    Max=ScalarParameter() {defaultValue=1, group=Item_Border, parameterName=gGradientMax}
   ) {materialFunction=&SmoothStep}.Result)
  n28 = Multiply(A=TextureCoordinate(), B=Constant2Vector() {g=1, r=1})
  n29 = Subtract(
   A=Abs(None=Add(A=@n28) {constB=-0.5})
   B=ScalarParameter() {defaultValue=0.45, group=Item_Border, parameterName=gStrokeThickness, sliderMax=1}
  )
  n30 = MaterialFunctionCall(
   Alpha=MaterialFunctionCall(Vector 2=Max(A=@n29) {constB=0}) {materialFunction=/Engine/Functions/Engine_MaterialFunctions02/Utility/VectorLength}.V2 Length
   Min=ScalarParameter() {group=Item_Border, parameterName=gStrokeMin}
   Max=ScalarParameter() {defaultValue=0.001, group=Item_Border, parameterName=gStrokeMax}
  ) {materialFunction=&SmoothStep}
  n31 = ScalarParameter() {defaultValue=1.25, group=Item_Type_Items, parameterName=gIcon-Scale}
  n32 = MaterialFunctionCall() {materialFunction=&GetUserInterfaceUV}
  n33 = ScalarParameter() {defaultValue=256, desc=The actual icon texel width. Passed in from code., group=Item_Icon, parameterName=IconActualTexelSizeX}
  n34 = ScalarParameter() {defaultValue=256, desc=The actual icon texel height. Passed in from code., group=Item_Icon, parameterName=IconActualTexelSizeY}
  n35 = ScalarParameter() {
   defaultValue=1
   desc="The minimum icon texel width that will be displayed at full size. If the icon is smaller than this size, it will be scaled down accordingly."
   group=Item_Icon
   parameterName=IconMinimumTexelSizeX
  }
  n36 = ScalarParameter() {
   defaultValue=1
   desc="The minimum icon texel height that will be displayed at full size. If the icon is smaller than this size, it will be scaled down accordingly."
   group=Item_Icon
   parameterName=IconMinimumTexelSizeY
  }
  n37 = Divide(A=AppendVector(A=@n33, B=@n34), B=AppendVector(A=@n35, B=@n36))
  n38 = ComponentMask(None=@n37) {r=true}
  n39 = ComponentMask(None=@n37) {g=true}
  n40 = MaterialFunctionCall(UVs=@n32.Normalized UV, Texture Scale=Clamp(None=Min(A=@n38, B=@n39))) {materialFunction=&ScaleUVsByCenter}
  n41 = Subtract(B=@n31)
  n42 = Multiply(
   A=@n40.UVs
   B=AppendVector(
    A=ScalarParameter() {defaultValue=1.067, parameterName=gIcon-Item-X Scale}
    B=ScalarParameter() {defaultValue=1.725, parameterName=gIcon-Item-Y Scale_1}
   )
  ) {constB=1.25}
  n43 = AppendVector(
   A=ScalarParameter() {defaultValue=-0.0333, parameterName=gIcon-Item-X Offset}
   B=ScalarParameter() {defaultValue=-0.3625, parameterName=gIcon-Item-Y Offset}
  )
  n44 = ScalarParameter() {defaultValue=0.7, parameterName=gIcon-Portrait-X Scale}
  n45 = Multiply(A=@n40.UVs, B=AppendVector(A=@n44, B=Multiply(A=@n44) {constB=1.6129})) {constB=1.25}
  n46 = AppendVector(A=Divide(A=Subtract(B=@n44)), B=ScalarParameter() {parameterName=gIcon-Portrait-Y Offset})
  n47 = LinearInterpolate(
   A=LinearInterpolate(A=Add(A=@n42, B=@n43), B=Add(A=@n45, B=@n46), Alpha=ScalarParameter() {defaultValue=1, parameterName=bIsHero, sliderMax=1})
   B=Add(A=Multiply(A=@n40.UVs, B=@n31) {constB=1.25}, B=Multiply(A=@n41) {constB=0.5})
   Alpha=ScalarParameter() {group=Item_Type, parameterName=bIsItem, sliderMax=1}
  ) {constAlpha=0}
  n48 = TextureSampleParameter2D(UVs=@n47) {
   group=Item_Icon
   parameterName=Icon
   samplerSource=SSM_Clamp_WorldGroupSettings
   texture=/Game/UI/Foundation/Textures/Icons/Heroes/Athena/Soldier/T-Soldier-HID-005-Athena-Commando-M-L
  }
  n49 = MaterialFunctionCall!(UVs=@n47, Alpha=@n48.A) {materialFunction=/Game/UI/Foundation/Items/ClampAlphaToTransparentBorder}
  n50 = AppendVector(A=Constant() {r=2.25}, B=Constant() {r=3.6383})
  n51 = Constant() {r=-1.125}
  n52 = Constant() {r=-0.8}
  n53 = Rotator(Coordinate=Add(A=Multiply(A=@n32.Normalized UV, B=@n50), B=AppendVector(A=@n51, B=@n52)), Time=Constant() {r=0.02}) {speed=6.2832}
  n54 = Add(A=@n53, B=Constant2Vector() {g=-1.125})
  n55 = MaterialFunctionCall(Alpha=OneMinus(None=ComponentMask(None=@n54) {g=true}), Min=Constant(), Max=Constant() {r=0.025}) {materialFunction=&SmoothStep}
  n56 = VectorParameter() {group=Item_Rarity_Colors, parameterName=Rarity 5, defaultValue=<rgba 0.0648 0.0931 0.1356 1>}
  n57 = MaterialFunctionCall(UVs=@n32.Normalized UV, CenterPosition=Constant2Vector() {g=0.5, r=0.5}, Radius=Constant() {r=1.2}, Density=Constant() {r=1.125}) {materialFunction=&RadialGradientExponential}
  n58 = VectorParameter() {parameterName=Rarity 2, defaultValue=<rgba 0.0574 0.0624 0.07 1>}
  n59 = LinearInterpolate(
   A=@n56.RGB
   B=Multiply(A=Multiply(A=@n57.RadialGradientExponential, B=@n58.RGB) {constB=2}) {constB=0.75}
   Alpha=@n57.RadialGradientExponential
  )
  n60 = MaterialFunctionCall(UVs=@n32.Normalized UV, CenterPosition=Constant2Vector() {g=0.5, r=0.5}, Radius=Constant() {r=1.125}, Density=Constant() {r=1}) {materialFunction=&RadialGradientExponential}
  n61 = Multiply(A=Multiply(A=@n60.RadialGradientExponential, B=@n58.RGB) {constB=2}) {constB=0.75}
  n62 = ComponentMask(None=Add(A=Constant2Vector() {g=1.13}, B=@n53)) {g=true}
  n63 = Constant() {r=0.975}
  n64 = Constant() {r=1}
  n65 = Clamp(None=DotProduct(
    A=MaterialFunctionCall(Alpha=@n62, Min=@n63, Max=@n64) {materialFunction=&SmoothStep}.Result
    B=@n55.Result
   ))
  n66 = LinearInterpolate(
   A=Desaturation(None=LinearInterpolate(A=@n56.RGB, B=@n61, Alpha=@n60.RadialGradientExponential), Fraction=ScalarParameter() {parameterName=Desat})
   B=Desaturation(None=@n59, Fraction=ScalarParameter() {parameterName=Desat_1})
   Alpha=LinearInterpolate(A=OneMinus(None=@n65), Alpha=ScalarParameter() {parameterName=bIsVariant, sliderMax=1})
  )
  n67 = LinearInterpolate(
   A=@n66
   B=LinearInterpolate(A=@n59, B=@n66, Alpha=Clamp(None=@n55.Result))
   Alpha=ScalarParameter() {defaultValue=1, parameterName=bIsHero, sliderMax=1}
  )
  n68 = LinearInterpolate(
   A=@n67
   B=@n48.RGB
   Alpha=LinearInterpolate(A=@n49.Result, B=LinearInterpolate(B=@n49.Result, Alpha=@n55.Result), Alpha=ScalarParameter() {defaultValue=1, parameterName=bIsHero})
  ) {constAlpha=0.8}
  n69 = ScalarParameter() {defaultValue=1, desc="Master visibility for the Icon. 0 = hidden. 1 = visible.", parameterName=bIsIconVisible, sliderMax=1}
  LinearInterpolate(
   A=LinearInterpolate(A=LinearInterpolate(A=@n67, B=@n68, Alpha=@n69) {constAlpha=0.8}, B=@n67, Alpha=ScalarParameter() {parameterName=bIsLoading}) {constAlpha=0.8}
   B=LinearInterpolate(A=@n24, B=LinearInterpolate(A=@n24, B=@n23.RGB, Alpha=@n27), Alpha=ScalarParameter() {group=Item_Border, parameterName=bAnimateBorder})
   Alpha=Clamp(None=Multiply(A=@n30.Result))
  )
  ScalarParameter() {defaultValue=1, group=Item_General, parameterName=gAlpha, sliderMax=1}
  ScalarParameter() {defaultValue=1.25, desc=Not Hooked up, parameterName=gIcon-Portrait-Y Scale}
  ScalarParameter() {defaultValue=0.1125, desc=Not Hooked up, parameterName=gIcon-Portrait-X Offset}
  TextureCoordinate()
 ]
}
