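# Translucent UI master material that procedurally draws an item-card stat background, judging by its name, without textures.
# A dark navy base blends toward ModColor, scaled by sModification, through an exponential radial glow whose density pulses over time.
# Slightly rotated 9-slice UVs, sine and SmoothStep build slanted edge masks and layered shapes, some tinted by ModColor.
# The bIsBacking parameter blends between the glowing fill and a plain dark backing.
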Material /FortniteCoreUI/UI/Foundation/Items/Materials/M-Item-Card-FortStat-BG-Master
graph {
 properties {
  bUsedWithStaticMesh=false
  blendMode=BLEND_Translucent
  materialDomain=MD_UI
  pixelDepthOffsetMode=PDOM_Legacy
  usedShadingModels=Default Lit
 }
 nodes [
  Constant() {r=1}
  n1 = Constant3Vector()
  n2 = Constant3Vector() {constant=<rgba 0.0137 0.0194 0.0356 1>}
  n3 = VectorParameter() {parameterName=ModColor, defaultValue=<rgba 0 1 1 1>}
  n4 = Sine(None=Time()) {period=2}
  n5 = LinearInterpolate(
   A=@n2
   B=Multiply(A=@n3.RGB, B=ScalarParameter() {defaultValue=1, parameterName=sModification, sliderMax=1})
   Alpha=MaterialFunctionCall(
    CenterPosition=Constant2Vector() {g=0.45, r=0.4}
    Radius=ScalarParameter() {defaultValue=0.6, parameterName=sGlowRadius}
    Density=LinearInterpolate(Alpha=@n4) {constA=0.75, constB=0.85}
   ) {materialFunction=&RadialGradientExponential}.RadialGradientExponential
  )
  n6 = MaterialFunctionCall() {materialFunction=&GetUserInterfaceUV}
  n7 = Constant() {r=0.48}
  n8 = ComponentMask(None=Add(A=Rotator(Coordinate=@n6."9-Slice UV (No Tiling)", Time=@n7) {speed=6.2832}, B=Constant2Vector() {r=-0.1})) {r=true}
  n9 = Constant()
  n10 = Constant() {r=0.015}
  n11 = Constant2Vector() {g=0.1}
  n12 = OneMinus(None=ComponentMask(None=Add(A=@n6."9-Slice UV (No Tiling)", B=@n11)) {g=true})
  n13 = Constant()
  n14 = Constant() {r=0.015}
  n15 = DotProduct(
   A=Clamp(None=MaterialFunctionCall(Alpha=@n8, Min=@n9, Max=@n10) {materialFunction=&SmoothStep}.Result)
   B=Clamp(None=MaterialFunctionCall(Alpha=@n12, Min=@n13, Max=@n14) {materialFunction=&SmoothStep}.Result)
  )
  LinearInterpolate(
   A=LinearInterpolate(A=@n1, B=LinearInterpolate(B=@n5, Alpha=@n15), Alpha=@n15) {constA=0.5}
   B=LinearInterpolate(A=@n1, B=@n2, Alpha=@n15) {constA=0.5}
   Alpha=ScalarParameter() {defaultValue=1, parameterName=bIsBacking, sliderMax=1}
  )
  n16 = Constant() {r=0.47}
  n17 = Sine(None=Add(A=Rotator(Coordinate=@n6."9-Slice UV (No Tiling)", Time=@n16) {speed=6.2832}, B=Constant2Vector() {r=-0.05})) {period=1.8}
  n18 = ComponentMask(None=Add(A=@n6."9-Slice UV (No Tiling)", B=Constant2Vector())) {g=true}
  n19 = LinearInterpolate(
   B=MaterialFunctionCall(Alpha=ComponentMask(None=@n17) {r=true}, Min=Constant() {r=0.15}, Max=Constant() {r=0.2}) {materialFunction=&SmoothStep}.Result
   Alpha=MaterialFunctionCall(Alpha=Sine(None=@n18) {period=2}, Min=Constant() {r=0.2}, Max=Constant() {r=0.25}) {materialFunction=&SmoothStep}.Result
  )
  n20 = Constant2Vector() {g=0.05}
  n21 = OneMinus(None=ComponentMask(None=Add(A=@n6."9-Slice UV (No Tiling)", B=@n20)) {g=true})
  n22 = Constant()
  n23 = Constant() {r=0.015}
  n24 = Constant() {r=0.48}
  n25 = ComponentMask(None=Add(A=Rotator(Coordinate=@n6."9-Slice UV (No Tiling)", Time=@n24) {speed=6.2832}, B=Constant2Vector() {r=-0.075})) {r=true}
  n26 = Constant()
  n27 = Constant() {r=0.015}
  n28 = DotProduct(
   A=Clamp(None=MaterialFunctionCall(Alpha=@n21, Min=@n22, Max=@n23) {materialFunction=&SmoothStep}.Result)
   B=Clamp(None=MaterialFunctionCall(Alpha=@n25, Min=@n26, Max=@n27) {materialFunction=&SmoothStep}.Result)
  )
  LinearInterpolate(A=@n19, B=LinearInterpolate(A=LinearInterpolate(Alpha=@n28) {constB=0.25}, Alpha=@n15))
  n29 = Constant() {r=0.46}
  n30 = Sine(None=Add(A=Rotator(Coordinate=@n6."9-Slice UV (No Tiling)", Time=@n29) {speed=6.2832}, B=Constant2Vector() {r=0.0125})) {period=1.8}
  n31 = Constant() {r=0.51}
  n32 = ComponentMask(None=Add(A=Rotator(Coordinate=@n6."9-Slice UV (No Tiling)", Time=@n31) {speed=6.2832}, B=Constant2Vector() {g=0.09})) {g=true}
  n33 = LinearInterpolate(
   B=MaterialFunctionCall(Alpha=ComponentMask(None=@n30) {r=true}, Min=Constant() {r=0.15}, Max=Constant() {r=0.2}) {materialFunction=&SmoothStep}.Result
   Alpha=MaterialFunctionCall(Alpha=Sine(None=@n32) {period=2}, Min=Constant() {r=0.2}, Max=Constant() {r=0.25}) {materialFunction=&SmoothStep}.Result
  ) {constB=0}
  LinearInterpolate(
   B=LinearInterpolate(A=Multiply(A=@n19, B=@n3.RGB) {constB=0.25}, B=Multiply(A=@n3.RGB) {constB=0.6}, Alpha=LinearInterpolate(A=@n19, Alpha=@n33) {constB=0})
   Alpha=@n19
  )
  TextureCoordinate()
 ]
}
