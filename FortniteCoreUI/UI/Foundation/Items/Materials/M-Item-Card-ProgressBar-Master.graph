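# Translucent UI master material for an item card progress bar, filled in the rarity color up to a segment-snapped progress value.
# Rounded progress times the segment-size parameter forms step masks along U for the filled range and the segment after it.
# That next segment shows an upgrade highlight, shaped by sine and smoothstep masks and pulsed over time.
# A scalar parameter blends the highlight in or out.
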Material /FortniteCoreUI/UI/Foundation/Items/Materials/M-Item-Card-ProgressBar-Master
graph {
 properties {
  bUsedWithStaticMesh=false
  blendMode=BLEND_Translucent
  materialDomain=MD_UI
  pixelDepthOffsetMode=PDOM_Legacy
  usedShadingModels=Default Lit
 }
 nodes [
  n1 = ScalarParameter() {parameterName=sProgress, sliderMax=10}
  n2 = ScalarParameter() {defaultValue=0.1, parameterName=sSegmentCount, sliderMax=1}
  n3 = TextureCoordinate()
  n4 = ComponentMask(None=@n3) {r=true}
  n5 = Step(Y=Multiply(A=Round(None=@n1), B=@n2), X=@n4)
  n6 = LinearInterpolate(A=VectorParameter() {parameterName=Rarity 1, defaultValue=<rgba 1 1 1 1>}.RGB, Alpha=@n5) {constB=0}
  n7 = MaterialFunctionCall(Value=ComponentMask(None=@n3) {r=true}, Period=@n2, Sine Phase=StaticBool()) {materialFunction=&LinearSine}
  n8 = MaterialFunctionCall(Alpha=@n7.Linear Sine, Min=Min(A=@n7.Linear Sine) {constB=0.15}, Max=Max(A=@n7.Linear Sine) {constB=0.2}) {materialFunction=&SmoothStep}
  n9 = ComponentMask(None=@n3) {g=true}
  n10 = Time()
  n11 = Multiply(
   A=Step(Y=ScalarParameter() {defaultValue=0.5, parameterName=gEdgeClamp, sliderMax=1}, X=Sine(None=@n9) {period=2})
   B=LinearInterpolate(Alpha=Sine(None=@n10) {period=2}) {constA=1, constB=1.5}
  )
  n12 = LinearInterpolate(
   B=VectorParameter() {parameterName=vColor, defaultValue=<rgba 0.995 0.916 0.031 1>}.RGB
   Alpha=Clamp(None=LinearInterpolate(B=@n8.Result, Alpha=@n11))
  )
  n13 = Add(A=@n1)
  n14 = Step(Y=Multiply(A=Round(None=@n13), B=@n2), X=@n4)
  LinearInterpolate(
   A=@n6
   B=LinearInterpolate(A=@n6, B=LinearInterpolate(A=@n12, Alpha=@n14) {constB=0}, Alpha=@n5)
   Alpha=ScalarParameter() {defaultValue=1, parameterName=bShowUpgrade, sliderMax=1}
  )
 ]
}
