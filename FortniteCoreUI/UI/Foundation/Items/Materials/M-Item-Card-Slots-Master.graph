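# Translucent UI master material for item card slot icons, switching between active, default, empty and dimmed looks through scalar state parameters.
# Active tints the icon with Color and adds a radial gradient glow, whose density pulses on a time sine when bPulsing is set and whose alpha fades via UV sine masks.
# Empty, on by default, replaces the icon with a semi-transparent grey bar built procedurally from smoothstepped UV sines.
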Material /FortniteCoreUI/UI/Foundation/Items/Materials/M-Item-Card-Slots-Master
graph {
 properties {
  bUsedWithStaticMesh=false
  blendMode=BLEND_Translucent
  materialDomain=MD_UI
  pixelDepthOffsetMode=PDOM_Legacy
  usedShadingModels=Default Lit
 }
 nodes [
  n1 = TextureSampleParameter2D() {
   parameterName=gIcon-Slot
   samplerSource=SSM_Clamp_WorldGroupSettings
   texture=/FortniteCoreUI/UI/Foundation/Textures/Icons/Heroes/Soldier/Perks/Icon-Soldier-CleanLiving-128
  }
  n2 = Power(Base=@n1.R)
  n3 = VectorParameter() {parameterName=Color, defaultValue=<rgba 0.7231 0.6445 0.063 1>}
  n4 = Sine(None=Time())
  n5 = ScalarParameter() {group=States, parameterName=bPulsing, sliderMax=1}
  n6 = MaterialFunctionCall(
   CenterPosition=Constant2Vector() {g=0.5, r=0.5}
   Radius=Constant() {r=0.6}
   Density=LinearInterpolate(
    A=ScalarParameter() {defaultValue=1.5, parameterName=Lerp Min}
    B=ScalarParameter() {defaultValue=3, parameterName=Lerp Max}
    Alpha=LinearInterpolate(B=@n4, Alpha=@n5) {constA=1}
   ) {constB=1.5}
  ) {materialFunction=&RadialGradientExponential}
  n7 = Power(Base=Multiply(A=@n6.RadialGradientExponential) {constB=0.125}) {constExponent=0.75}
  n8 = Add(A=Add(A=Multiply(A=@n2, B=@n3.RGB), B=@n7), B=LinearInterpolate(B=@n1.A, Alpha=@n7)) {constB=0}
  n9 = ScalarParameter() {group=States, parameterName=bActive, sliderMax=1}
  n10 = ScalarParameter() {group=States, parameterName=bDefaultState, sliderMax=1}
  n11 = ScalarParameter() {defaultValue=1, group=States, parameterName=bSlot_Empty, sliderMax=1}
  LinearInterpolate(
   A=LinearInterpolate(A=LinearInterpolate(A=@n1.RGB, B=@n8, Alpha=@n9), B=@n1.RGB, Alpha=@n10)
   B=Constant3Vector() {constant=<rgba 0.5 0.5 0.5 0>}
   Alpha=@n11
  )
  n12 = Multiply(A=@n1.A) {constB=0.25}
  n13 = Power(Base=@n6.RadialGradientExponential) {constExponent=0.6}
  n14 = ScalarParameter() {defaultValue=0.5, parameterName=s Glow Alpha-Surpress}
  n15 = Sine(None=TextureCoordinate()) {period=2}
  n16 = LinearInterpolate(B=ComponentMask(None=@n15) {r=true}, Alpha=ComponentMask(None=@n15) {g=true})
  n17 = ScalarParameter() {defaultValue=0.75, parameterName=sGlow Edge Clamp}
  n18 = Add(A=@n1.A, B=LinearInterpolate(B=Multiply(A=@n13, B=@n14), Alpha=Power(Base=@n16, Exp=@n17) {constExponent=1.5}))
  n19 = TextureCoordinate()
  n20 = Constant() {r=0.9925}
  n21 = MaterialFunctionCall(Alpha=ComponentMask(None=Sine(None=@n19) {period=3}) {g=true}, Min=@n20, Max=@n20) {materialFunction=&SmoothStep}
  n22 = Constant() {r=0.85}
  n23 = MaterialFunctionCall(Alpha=ComponentMask(None=Sine(None=@n19) {period=2}) {r=true}, Min=@n22, Max=@n22) {materialFunction=&SmoothStep}
  LinearInterpolate(
   A=LinearInterpolate(A=LinearInterpolate(A=@n12, B=@n18, Alpha=@n9), B=@n1.A, Alpha=@n10)
   B=Multiply(A=DotProduct(A=@n21.Result, B=@n23.Result)) {constB=0.75}
   Alpha=@n11
  )
  Multiply(A=VectorParameter() {group=Vector Parameter Values, parameterName=Rarity 5, defaultValue=<rgba 0.0006 0.0331 0.0006 1>}.RGB) {constB=0.125}
 ]
}
