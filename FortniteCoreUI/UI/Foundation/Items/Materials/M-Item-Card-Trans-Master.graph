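# Translucent UI master material for item cards, specifically the Transform item type, tinted by four rarity color parameters.
# Rotated bands frame a sine-generated grid, masked by a transform background texture and radial gradient, with an animated sweep highlight.
# The Icon texture composites on top with an offset drop shadow from its alpha.
# A procedural rounded stroke adds a gradient border, optionally animated through bAnimateBorder, gRotation and gTime.
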Material /FortniteCoreUI/UI/Foundation/Items/Materials/M-Item-Card-Trans-Master
graph {
 properties {
  bUsedWithStaticMesh=false
  blendMode=BLEND_Translucent
  materialDomain=MD_UI
  pixelDepthOffsetMode=PDOM_Legacy
  usedShadingModels=Default Lit
 }
 nodes [
  n1 = Constant2Vector() {g=1}
  n2 = Multiply() {constB=1.4}
  n3 = MaterialFunctionCall(Alpha=ComponentMask(None=Add(A=@n1, B=@n2)) {g=true}, Min=Constant(), Max=Constant() {r=0.0125}) {materialFunction=&SmoothStep}
  n4 = Add(A=@n2, B=Constant2Vector() {g=-1})
  n5 = MaterialFunctionCall(Alpha=OneMinus(None=ComponentMask(None=@n4) {g=true}), Min=Constant() {r=0.9875}, Max=Constant() {r=1}) {materialFunction=&SmoothStep}
  n6 = DotProduct(A=@n3.Result, B=@n5.Result)
  n7 = Constant2Vector() {r=-0.124}
  n8 = OneMinus(None=ComponentMask(None=Add(A=@n7, B=@n2)) {r=true})
  n9 = Constant() {r=0.9985}
  n10 = Constant() {r=1}
  n11 = Constant2Vector() {r=0.124}
  n12 = MaterialFunctionCall(Alpha=ComponentMask(None=Add(A=@n2, B=@n11)) {r=true}, Min=Constant(), Max=Constant() {r=0.0125}) {materialFunction=&SmoothStep}
  n13 = DotProduct(
   A=@n6
   B=DotProduct(
    A=MaterialFunctionCall(Alpha=@n8, Min=@n9, Max=@n10) {materialFunction=&SmoothStep}.Result
    B=@n12.Result
   )
  )
  n14 = Constant2Vector() {g=-1, r=-0.275}
  n15 = Constant() {r=0.62}
  n16 = MaterialFunctionCall(UVs=@n2, CenterPosition=@n14, Radius=Constant() {r=0.217}) {materialFunction=&RadialGradientExponential}
  n17 = Constant() {r=0.375}
  n18 = Constant() {r=0.475}
  n19 = DotProduct(
   A=MaterialFunctionCall(
    Alpha=MaterialFunctionCall(UVs=@n2, CenterPosition=@n14, Radius=@n15) {materialFunction=&RadialGradientExponential}.RadialGradientExponential
    Min=Constant() {r=0.5}
    Max=Constant() {r=0.525}
   ) {materialFunction=&SmoothStep}.Result
   B=OneMinus(None=MaterialFunctionCall(Alpha=@n16.RadialGradientExponential, Min=@n17, Max=@n18) {materialFunction=&SmoothStep}.Result)
  )
  n20 = Add(A=Constant2Vector() {g=-0.124}, B=@n2)
  n21 = MaterialFunctionCall(Alpha=OneMinus(None=ComponentMask(None=@n20) {g=true}), Min=Constant() {r=0.9875}, Max=Constant() {r=1}) {materialFunction=&SmoothStep}
  n22 = Add(A=Constant2Vector() {r=0.425}, B=@n2)
  n23 = MaterialFunctionCall(Alpha=OneMinus(None=ComponentMask(None=@n22) {r=true}), Min=Constant() {r=0.9875}, Max=Constant() {r=1}) {materialFunction=&SmoothStep}
  n24 = ComponentMask(None=Add(A=@n2, B=Constant2Vector() {r=0.675})) {r=true}
  n25 = Constant()
  n26 = Constant() {r=0.0125}
  n27 = DotProduct(
   A=@n21.Result
   B=DotProduct(
    A=@n23.Result
    B=MaterialFunctionCall(Alpha=@n24, Min=@n25, Max=@n26) {materialFunction=&SmoothStep}.Result
   )
  )
  n28 = Constant2Vector() {g=0.124}
  n29 = MaterialFunctionCall(Alpha=ComponentMask(None=Add(A=@n2, B=@n28)) {g=true}, Min=Constant(), Max=Constant() {r=0.0125}) {materialFunction=&SmoothStep}
  n30 = Add(A=Constant2Vector() {r=-0.655}, B=@n2)
  n31 = MaterialFunctionCall(Alpha=OneMinus(None=ComponentMask(None=@n30) {r=true}), Min=Constant() {r=0.985}, Max=Constant() {r=1}) {materialFunction=&SmoothStep}
  n32 = Constant2Vector() {r=-0.415}
  n33 = MaterialFunctionCall(Alpha=ComponentMask(None=Add(A=@n2, B=@n32)) {r=true}, Min=Constant(), Max=Constant() {r=0.0125}) {materialFunction=&SmoothStep}
  n34 = DotProduct(A=@n29.Result, B=DotProduct(A=@n12.Result, B=DotProduct(A=@n31.Result, B=@n33.Result)))
  n35 = Constant2Vector() {g=1, r=0.275}
  n36 = Constant() {r=0.62}
  n37 = MaterialFunctionCall(UVs=@n2, CenterPosition=@n35, Radius=Constant() {r=0.217}) {materialFunction=&RadialGradientExponential}
  n38 = Constant() {r=0.375}
  n39 = Constant() {r=0.475}
  n40 = DotProduct(
   A=MaterialFunctionCall(
    Alpha=MaterialFunctionCall(UVs=@n2, CenterPosition=@n35, Radius=@n36) {materialFunction=&RadialGradientExponential}.RadialGradientExponential
    Min=Constant() {r=0.5}
    Max=Constant() {r=0.525}
   ) {materialFunction=&SmoothStep}.Result
   B=OneMinus(None=MaterialFunctionCall(Alpha=@n37.RadialGradientExponential, Min=@n38, Max=@n39) {materialFunction=&SmoothStep}.Result)
  )
  n41 = Add(A=LinearInterpolate(A=@n19, B=@n27, Alpha=Clamp(None=@n6)), B=LinearInterpolate(A=@n40, B=@n34, Alpha=Clamp(None=@n6)))
  n42 = DotProduct(A=@n21.Result, B=@n29.Result)
  n43 = DotProduct(A=@n42, B=@n33.Result)
  n44 = DotProduct(A=@n42, B=@n23.Result)
  n45 = Clamp(None=Add(A=Add(A=@n43, B=@n44), B=Add(A=@n13, B=@n41)))
  n46 = Clamp(None=Add(A=DotProduct(A=@n43, B=@n34), B=DotProduct(A=@n44, B=@n27)))
  n47 = Constant2Vector() {g=-0.125, r=-0.675}
  n48 = Constant() {r=0.3875}
  n49 = DotProduct(
   A=MaterialFunctionCall(
    Alpha=MaterialFunctionCall(UVs=@n2, CenterPosition=@n47, Radius=@n48) {materialFunction=&RadialGradientExponential}.RadialGradientExponential
    Min=Constant() {r=0.45}
    Max=Constant() {r=0.5}
   ) {materialFunction=&SmoothStep}.Result
   B=@n44
  )
  n50 = Constant2Vector() {g=0.125, r=0.675}
  n51 = Constant() {r=0.3875}
  n52 = DotProduct(
   A=MaterialFunctionCall(
    Alpha=MaterialFunctionCall(UVs=@n2, CenterPosition=@n50, Radius=@n51) {materialFunction=&RadialGradientExponential}.RadialGradientExponential
    Min=Constant() {r=0.45}
    Max=Constant() {r=0.5}
   ) {materialFunction=&SmoothStep}.Result
   B=@n43
  )
  Clamp(None=Add(A=Subtract(A=@n45, B=@n46), B=Add(A=@n49, B=@n52)))
  DotProduct(A=OneMinus(None=@n3.Result), B=@n19)
  DotProduct(A=OneMinus(None=@n5.Result), B=@n40)
  n53 = Constant() {desc="Rounded Constant for (X/Y)", r=0.9}
  n54 = Constant() {r=1.4555}
  n55 = MaterialFunctionCall() {materialFunction=&GetUserInterfaceUV}
  n56 = Add(
   A=Multiply(A=@n55.Normalized UV, B=AppendVector(A=@n53, B=@n54))
   B=AppendVector(A=Constant() {desc=X Offset required to Centre X tile of 0.62, r=0.05}, B=Constant() {r=-0.2275})
  )
  n57 = Constant() {r=0.125}
  n58 = TextureSample(UVs=Multiply(A=Rotator(Coordinate=@n56, Time=@n57) {speed=6.2831})) {samplerSource=SSM_Clamp_WorldGroupSettings, texture=/FortniteCoreUI/UI/Foundation/Items/Textures/T-Transform-Bgnd}
  n59 = Ceil(None=@n58.RGB)
  n60 = Constant() {r=0.125}
  n61 = Floor(None=@n58.RGB)
  n62 = Constant() {r=0.25}
  n63 = MaterialFunctionCall(Alpha=@n58.RGB, Min=Min(A=Ceil(None=@n58.RGB), B=Constant() {r=0.4}), Max=Max(A=Floor(None=@n58.RGB), B=Constant() {r=0.5})) {materialFunction=&SmoothStep}
  LinearInterpolate(
   A=MaterialFunctionCall(Alpha=@n58.RGB, Min=Min(A=@n59, B=@n60), Max=Max(A=@n61, B=@n62)) {materialFunction=&SmoothStep}.Result
   Alpha=@n63.Result
  )
  n64 = VectorParameter() {group=Item_Rarity_Colors, parameterName=Rarity 4, defaultValue=<rgba 0.0006 0.0782 0.0009 1>}
  n65 = VectorParameter() {group=Item_Rarity_Colors, parameterName=Rarity 2, defaultValue=<rgba 0.1195 0.521 0 1>}
  n66 = ComponentMask(None=Rotator(Coordinate=TextureCoordinate(), Time=ScalarParameter() {defaultValue=4.25, group=Item_Border, parameterName=gGradientAngle})) {r=true}
  n67 = ScalarParameter() {group=Item_Border, parameterName=gGradientMin}
  n68 = ScalarParameter() {defaultValue=1, group=Item_Border, parameterName=gGradientMax}
  n69 = Clamp(None=DotProduct(
    A=Add(A=@n66) {constB=0}
    B=MaterialFunctionCall(Alpha=@n66, Min=@n67, Max=@n68) {materialFunction=&SmoothStep}.Result
   ))
  n70 = LinearInterpolate(A=@n64.RGB, B=@n65.RGB, Alpha=@n69)
  n71 = Add(A=VectorParameter() {group=Item_Rarity_Colors, parameterName=Rarity 1, defaultValue=<rgba 0.3564 0.9911 0 1>}.RGB) {constB=0.25}
  n72 = TextureCoordinate()
  n73 = ScalarParameter() {defaultValue=-8.5, group=Item_Border, parameterName=gRotation}
  n74 = Time()
  n75 = ScalarParameter() {defaultValue=0.5, group=Item_General, parameterName=gTime}
  n76 = Frac(None=Add(A=Rotator(Coordinate=@n72, Time=@n73), B=Multiply(A=@n74, B=@n75)))
  n77 = ComponentMask(None=Abs(None=Add(A=@n76) {constB=-0.5})) {r=true}
  n78 = Clamp(None=MaterialFunctionCall(Alpha=@n77, Min=@n67, Max=@n68) {materialFunction=&SmoothStep}.Result)
  n79 = ScalarParameter() {group=Item_Border, parameterName=bAnimateBorder}
  n80 = TextureCoordinate()
  n81 = Constant2Vector() {g=1, r=1}
  n82 = Abs(None=Add(A=Multiply(A=@n80, B=@n81)) {constB=-0.5})
  n83 = ScalarParameter() {defaultValue=0.45, group=Item_Border, parameterName=gStrokeThickness, sliderMax=1}
  n84 = MaterialFunctionCall(Vector 2=Max(A=Subtract(A=@n82, B=@n83)) {constB=0}) {materialFunction=/Engine/Functions/Engine_MaterialFunctions02/Utility/VectorLength}
  n85 = ScalarParameter() {group=Item_Border, parameterName=gStrokeMin}
  n86 = ScalarParameter() {defaultValue=0.001, group=Item_Border, parameterName=gStrokeMax}
  n87 = VectorParameter() {group=Item_Rarity_Colors, parameterName=Rarity 3, defaultValue=<rgba 0 0.2159 0.0015 1>}
  n88 = VectorParameter() {group=Item_Rarity_Colors, parameterName=Rarity 2, defaultValue=<rgba 0.1195 0.521 0 1>}
  n89 = MaterialFunctionCall() {materialFunction=&GetUserInterfaceUV}
  n90 = MaterialFunctionCall(UVs=@n89.Normalized UV, CenterPosition=Constant2Vector() {g=0.5, r=0.5}, Radius=Constant() {r=0.85}, Density=Constant() {r=0.8}) {materialFunction=&RadialGradientExponential}
  n91 = Constant() {desc="Rounded Constant for (X/Y)", r=2}
  n92 = Constant() {r=3.2344}
  n93 = Add(
   A=Multiply(A=@n55.Normalized UV, B=AppendVector(A=@n91, B=@n92))
   B=AppendVector(A=Constant() {desc=X Offset required to Centre X tile of 0.62, r=-1.22}, B=Constant() {r=-1.135})
  )
  n94 = Constant() {r=0.125}
  n95 = Sine(None=Add(A=Rotator(Coordinate=@n93, Time=@n94) {speed=6.2832}) {constB=0.03}) {period=0.3}
  n96 = ComponentMask(None=@n95) {r=true}
  n97 = ScalarParameter() {defaultValue=0.98, desc=Grid Line Clamp Min, group=Item_Type_Transform, parameterName=sTransform Grid Min}
  n98 = ScalarParameter() {defaultValue=1, desc=Grid Line Clamp Max, group=Item_Type_Transform, parameterName=sTransform Grid Max}
  n99 = ComponentMask(None=@n95) {g=true}
  n100 = Add(
   A=MaterialFunctionCall(Alpha=@n96, Min=Min(A=@n96, B=@n97), Max=Max(A=@n96, B=@n98)) {materialFunction=&SmoothStep}.Result
   B=MaterialFunctionCall(Alpha=@n99, Min=Min(A=@n99, B=@n97), Max=Max(A=@n99, B=@n98)) {materialFunction=&SmoothStep}.Result
  )
  n101 = LinearInterpolate(B=LinearInterpolate(A=Clamp(None=@n100), Alpha=@n63.Result), Alpha=@n90.RadialGradientExponential)
  n102 = VectorParameter() {group=Item_Rarity_Colors, parameterName=Rarity 3, defaultValue=<rgba 0 0.2159 0.0015 1>}
  n103 = LinearInterpolate(
   A=LinearInterpolate(A=Constant3Vector() {constant=<rgba 0 0 0 1>}, B=Multiply(A=@n102.RGB) {constB=0.6}, Alpha=@n90.RadialGradientExponential)
   B=Multiply(A=LinearInterpolate(A=@n87.RGB, B=@n88.RGB, Alpha=@n90.RadialGradientExponential)) {constB=0.75}
   Alpha=@n101
  )
  n104 = Constant() {r=0.25}
  n105 = MaterialFunctionCall(
   Alpha=@n77
   Min=ScalarParameter() {defaultValue=0.25, desc=Schematic Animation Clamp Min, group=Item_Type_Transform, parameterName=gMaskGradientMin}
   Max=ScalarParameter() {defaultValue=0.75, desc=Schematic Animation Clamp Max, group=Item_Type_Transform, parameterName=gMaskGradientMax}
  ) {materialFunction=&SmoothStep}
  n106 = DotProduct(A=Clamp(None=Clamp(None=@n105.Result)), B=@n101)
  n107 = VectorParameter() {group=Item_Rarity_Colors, parameterName=Rarity 3, defaultValue=<rgba 0 0.2159 0.0015 1>}
  n108 = LinearInterpolate(A=@n69, B=@n78, Alpha=@n79)
  n109 = Constant2Vector() {g=1.325}
  n110 = Constant() {r=2.25}
  n111 = Constant() {r=3.6383}
  n112 = Add(A=Multiply(A=@n89.Normalized UV, B=AppendVector(A=@n110, B=@n111)), B=AppendVector(A=Constant() {r=-1.125}, B=Constant() {r=-0.8}))
  n113 = Rotator(Coordinate=@n112, Time=Constant() {r=0.02}) {speed=6.2832}
  n114 = MaterialFunctionCall(Alpha=ComponentMask(None=Add(A=@n109, B=@n113)) {g=true}, Min=Constant() {r=0.975}, Max=Constant() {r=1}) {materialFunction=&SmoothStep}
  n115 = Add(A=@n113, B=Constant2Vector() {g=-1.23})
  n116 = MaterialFunctionCall(Alpha=OneMinus(None=ComponentMask(None=@n115) {g=true}), Min=Constant(), Max=Constant() {r=0.025}) {materialFunction=&SmoothStep}
  n117 = Rotator(Coordinate=@n112, Time=Constant() {r=0.0347}) {speed=6.2832}
  n118 = ComponentMask(None=Add(A=Constant2Vector() {g=1.25}, B=@n117)) {g=true}
  n119 = Constant() {r=0.975}
  n120 = Constant() {r=1}
  n121 = Constant2Vector() {g=-1}
  n122 = OneMinus(None=ComponentMask(None=Add(A=@n117, B=@n121)) {g=true})
  n123 = Constant()
  n124 = Constant() {r=0.025}
  n125 = LinearInterpolate(
   A=LinearInterpolate(
    A=LinearInterpolate(A=@n107.RGB, B=@n65.RGB, Alpha=@n108)
    B=Multiply(A=@n64.RGB) {constB=0.75}
    Alpha=DotProduct(A=@n114.Result, B=@n116.Result)
   )
   B=Add(A=Desaturation(None=@n103, Fraction=@n104), B=LinearInterpolate(B=@n88.RGB, Alpha=@n106))
   Alpha=DotProduct(
    A=MaterialFunctionCall(Alpha=@n118, Min=@n119, Max=@n120) {materialFunction=&SmoothStep}.Result
    B=MaterialFunctionCall(Alpha=@n122, Min=@n123, Max=@n124) {materialFunction=&SmoothStep}.Result
   )
  )
  n126 = ScalarParameter() {defaultValue=1.0667, desc="Card Proportion (X=100%)", group=Item_Type_Transform, parameterName=gIcon-Transform-X Scale}
  n127 = ScalarParameter() {defaultValue=1.725, desc="Card Proportion (Y +27.5%)", group=Item_Type_Transform, parameterName=gIcon-Transform-Y Scale}
  n128 = ScalarParameter() {defaultValue=-0.0335, desc="X Offset for Card Proportion (X=100%)", group=Item_Type_Transform, parameterName=gIcon-Transform-X-Offset}
  n129 = Add(
   A=Multiply(A=@n89.Normalized UV, B=AppendVector(A=@n126, B=@n127))
   B=AppendVector(A=@n128, B=ScalarParameter() {defaultValue=-0.3625, group=Item_Type_Transform, parameterName=gIcon-Transform-Y-Offset})
  )
  n130 = Add(A=@n129, B=Constant2Vector() {g=-0.025, r=-0.0125}) {constB=-0.05}
  n131 = TextureSampleParameter2D(UVs=@n130) {
   group=Item_Icon
   parameterName=Icon
   samplerSource=SSM_Clamp_WorldGroupSettings
   texture=/Game/Widgets/SMyTownMapWidget/Common/BonusItemBackground
  }
  n132 = Multiply(A=MaterialFunctionCall!(UVs=@n130, Alpha=@n131.A) {materialFunction=/Game/UI/Foundation/Items/ClampAlphaToTransparentBorder}.Result) {constB=0.35}
  n133 = TextureSampleParameter2D(UVs=@n129) {
   group=Item_Icon
   parameterName=Icon
   samplerSource=SSM_Clamp_WorldGroupSettings
   texture=&BonusItemBackground
  }
  LinearInterpolate(
   A=LinearInterpolate(
    A=LinearInterpolate(A=@n125, Alpha=@n132) {constB=0}
    B=@n133.RGB
    Alpha=MaterialFunctionCall(UVs=@n129, Alpha=@n133.A) {materialFunction=&ClampAlphaToTransparentBorder}.Result
   ) {constAlpha=0.8}
   B=LinearInterpolate(A=@n70, B=LinearInterpolate(A=@n70, B=@n71, Alpha=@n78), Alpha=@n79)
   Alpha=Multiply(A=MaterialFunctionCall(Alpha=@n84.V2 Length, Min=@n85, Max=@n86) {materialFunction=&SmoothStep}.Result)
  )
  ScalarParameter() {defaultValue=1, group=Item_General, parameterName=gAlpha, sliderMax=1}
 ]
}
