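# Translucent UI master material that draws an item tile, named for llama items, with rarity-colored backing, icon and border.
# A radial rarity gradient, desaturated by Desat, sits under the Icon and its offset alpha-based drop shadow.
# A small masked Sub-Icon darkens the composite, then a box-distance stroke applies a rotated gradient border with an optional time-scrolled highlight.
# A separate sphere-gradient pattern branch and several loose nodes feed nothing in that composite.
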Material /FortniteCoreUI/UI/Foundation/Items/Materials/M-Item-Llama-Inst-Master
graph {
 properties {
  bUsedWithStaticMesh=false
  blendMode=BLEND_Translucent
  materialDomain=MD_UI
  pixelDepthOffsetMode=PDOM_Legacy
  usedShadingModels=Default Lit
 }
 nodes [
  Constant2Vector() {g=-0.85, r=0.65}
  Constant3Vector() {constant=<rgba 0 0 0 1>}
  Constant() {r=0.9}
  Constant() {r=1.8}
  n1 = VectorParameter() {group=Item_Rarity_Colors, parameterName=Rarity 2, defaultValue=<rgba 0.1195 0.521 0 1>}
  n2 = MaterialFunctionCall() {materialFunction=&GetUserInterfaceUV}
  n3 = ComponentMask(None=@n2.Normalized UV) {r=true}
  n4 = LinearInterpolate(A=VectorParameter() {group=Item_Rarity_Colors, parameterName=Rarity 3, defaultValue=<rgba 0 0.2159 0.0015 1>}.RGB, B=@n1.RGB, Alpha=@n3)
  n5 = LinearInterpolate(
   A=VectorParameter() {group=Item_Rarity_Colors, parameterName=Rarity 4, defaultValue=<rgba 0.0006 0.0782 0.0009 1>}.RGB
   B=Multiply(A=@n1.RGB) {constB=0.75}
   Alpha=@n3
  )
  n6 = LinearInterpolate(
   A=@n2.Normalized UV
   B=Multiply(A=@n2.Normalized UV, B=Constant2Vector() {g=0.7, r=1})
   Alpha=ScalarParameter() {group=ItemType, parameterName=bIsItem, sliderMax=1}
  )
  n7 = Constant2Vector() {g=1, r=0.45}
  n8 = Constant() {r=0.5}
  n9 = Constant() {r=0.502}
  n10 = MaterialFunctionCall(
   Alpha=MaterialFunctionCall!(
    UVs=Multiply(A=@n6, B=@n7)
    CenterPosition=Constant2Vector() {g=-1.6, r=-0.85}
    Radius=ScalarParameter() {defaultValue=3, group=Item_Type_CheckerBacking, parameterName=gInner Circle_a}
   ) {materialFunction=/Engine/Functions/Engine_MaterialFunctions01/Gradient/SphereGradient-2D}.Result 0-1
   Min=@n8
   Max=@n9
  ) {materialFunction=&SmoothStep}
  n11 = Constant2Vector() {g=1, r=0.5}
  n12 = MaterialFunctionCall(
   Alpha=MaterialFunctionCall(
    UVs=Multiply(A=@n6, B=@n11)
    CenterPosition=Constant2Vector() {g=2.5, r=1.75}
    Radius=ScalarParameter() {defaultValue=3, group=Item_Type_CheckerBacking, parameterName=gInner Circle_b}
   ) {materialFunction=&SphereGradient-2D}.Result 0-1
   Min=@n8
   Max=@n9
  ) {materialFunction=&SmoothStep}
  n13 = MaterialFunctionCall(
   Alpha=MaterialFunctionCall(
    UVs=@n6
    CenterPosition=Constant2Vector() {g=1.5, r=-1.6}
    Radius=ScalarParameter() {defaultValue=3, group=Item_Type_CheckerBacking, parameterName=gInner Circle_c}
   ) {materialFunction=&SphereGradient-2D}.Result 0-1
   Min=@n8
   Max=@n9
  ) {materialFunction=&SmoothStep}
  n14 = DotProduct(A=@n12.Result, B=@n13.Result)
  n15 = OneMinus(None=@n12.Result)
  n16 = OneMinus(None=@n13.Result)
  n17 = Clamp(None=Add(A=DotProduct(A=@n10.Result, B=@n14), B=DotProduct(A=@n15, B=@n16)))
  n18 = OneMinus()
  Desaturation(None=LinearInterpolate(A=LinearInterpolate(A=@n4, B=@n5, Alpha=@n17), Alpha=Clamp(None=@n18)))
  n19 = ScalarParameter() {defaultValue=1.125, group=Item_Type_Items, parameterName=gIcon-Scale}
  n20 = MaterialFunctionCall() {materialFunction=&GetUserInterfaceUV}
  n21 = Add(A=Multiply(A=@n20.Normalized UV, B=@n19) {constB=1.25}, B=Multiply(A=Subtract(B=@n19)) {constB=0.5})
  n22 = Add(A=@n21, B=Constant2Vector() {g=-0.025, r=-0.0125}) {constB=-0.05}
  n23 = TextureSampleParameter2D(UVs=@n22) {
   group=Item_Icon
   parameterName=Icon
   samplerSource=SSM_Clamp_WorldGroupSettings
   texture=/Game/Widgets/SMyTownMapWidget/Common/BonusItemBackground
  }
  n24 = VectorParameter() {group=Item_Rarity_Colors, parameterName=Rarity 1, defaultValue=<rgba 0.3564 0.9911 0 1>}
  n25 = MaterialFunctionCall() {materialFunction=&GetUserInterfaceUV}
  n26 = ComponentMask(None=Rotator(Coordinate=@n25.Normalized UV, Time=ScalarParameter() {defaultValue=4.25, group=Item_Border, parameterName=gGradientAngle})) {r=true}
  n27 = Add(A=@n26) {constB=0}
  n28 = ScalarParameter() {group=Item_Border, parameterName=gGradientMin}
  n29 = ScalarParameter() {defaultValue=1, group=Item_Border, parameterName=gGradientMax}
  n30 = MaterialFunctionCall(Alpha=@n26, Min=@n28, Max=@n29) {materialFunction=&SmoothStep}
  n31 = LinearInterpolate(
   A=VectorParameter() {group=Item_Rarity_Colors, parameterName=Rarity 3, defaultValue=<rgba 0 0.2159 0.0015 1>}.RGB
   B=Multiply(A=@n24.RGB) {constB=0.85}
   Alpha=Clamp(None=DotProduct(A=@n27, B=@n30.Result))
  )
  n32 = Rotator(Coordinate=@n25.Normalized UV, Time=ScalarParameter() {defaultValue=-8.5, group=Item_Border, parameterName=gRotation})
  n33 = Multiply(A=Time(), B=ScalarParameter() {defaultValue=0.25, group=Item_General, parameterName=gTime})
  n34 = Add(A=Frac(None=Add(A=@n32, B=@n33))) {constB=-0.5}
  n35 = MaterialFunctionCall(Alpha=ComponentMask(None=Abs(None=@n34)) {r=true}, Min=@n28, Max=@n29) {materialFunction=&SmoothStep}
  n36 = LinearInterpolate(
   A=@n31
   B=LinearInterpolate(A=@n31, B=Add(A=@n24.RGB) {constB=0.25}, Alpha=Clamp(None=@n35.Result))
   Alpha=ScalarParameter() {group=Item_Border, parameterName=bAnimateBorder}
  )
  n37 = MaterialFunctionCall(
   UVs=@n20.Normalized UV
   CenterPosition=Constant2Vector() {g=0.5, r=0.5}
   Radius=ScalarParameter() {defaultValue=0.75, group=ItemType, parameterName=ItemBackRadius}
   Density=ScalarParameter() {defaultValue=1.3826, group=ItemType, parameterName=ItemBackDensity}
  ) {materialFunction=&RadialGradientExponential}
  n38 = LinearInterpolate(
   A=VectorParameter() {group=Item_Rarity_Colors, parameterName=Rarity 4, defaultValue=<rgba 0.0006 0.0782 0.0009 1>}.RGB
   B=VectorParameter() {group=Item_Rarity_Colors, parameterName=Rarity 2, defaultValue=<rgba 0.1195 0.521 0 1>}.RGB
   Alpha=@n37.RadialGradientExponential
  )
  n39 = ScalarParameter() {parameterName=Desat}
  n40 = LinearInterpolate(
   A=LinearInterpolate(A=@n36, B=Desaturation(None=@n38, Fraction=@n39)) {constAlpha=0.75}
   Alpha=Multiply(A=MaterialFunctionCall!(UVs=@n22, Alpha=@n23.A) {materialFunction=/Game/UI/Foundation/Items/ClampAlphaToTransparentBorder}.Result) {constB=0.35}
  ) {constB=0}
  n41 = TextureSampleParameter2D(UVs=@n21) {
   group=Item_Icon
   parameterName=Icon
   samplerSource=SSM_Clamp_WorldGroupSettings
   texture=&BonusItemBackground
  }
  n42 = MaterialFunctionCall(UVs=@n21, Alpha=@n41.A) {materialFunction=&ClampAlphaToTransparentBorder}
  n43 = Add(
   A=MaterialFunctionCall(UVs=@n21, Texture Scale=Constant() {r=0.1406}) {materialFunction=&ScaleUVsByCenter}.UVs
   B=AppendVector(
    A=ScalarParameter() {defaultValue=0.65, parameterName=Small Icon X offset}
    B=ScalarParameter() {defaultValue=-0.8, parameterName=Small Icon Y offset}
   )
  )
  n44 = TextureSampleParameter2D(UVs=@n43) {
   group=Item_Icon
   parameterName=Sub-Icon
   samplerSource=SSM_Wrap_WorldGroupSettings
   texture=&BonusItemBackground
  }
  n45 = Subtract(A=Abs(None=Add(A=@n43) {constB=-0.5}), B=ScalarParameter() {defaultValue=0.45, group=Item_Border, parameterName=gMaskThickness, sliderMax=1})
  n46 = MaterialFunctionCall(
   Alpha=MaterialFunctionCall(Vector 2=Max(A=@n45) {constB=0}) {materialFunction=/Engine/Functions/Engine_MaterialFunctions02/Utility/VectorLength}.V2 Length
   Min=ScalarParameter() {group=Item_Border, parameterName=gMaskMin}
   Max=ScalarParameter() {defaultValue=0.001, group=Item_Border, parameterName=gMaskMax}
  ) {materialFunction=&SmoothStep}
  n47 = OneMinus(None=Clamp(None=Multiply(A=@n46.Result)))
  n48 = Multiply(A=TextureCoordinate(), B=Constant2Vector() {g=1, r=1})
  n49 = Subtract(
   A=Abs(None=Add(A=@n48) {constB=-0.5})
   B=ScalarParameter() {defaultValue=0.45, group=Item_Border, parameterName=gStrokeThickness, sliderMax=1}
  )
  n50 = MaterialFunctionCall(
   Alpha=MaterialFunctionCall(Vector 2=Max(A=@n49) {constB=0}) {materialFunction=&VectorLength}.V2 Length
   Min=ScalarParameter() {group=Item_Border, parameterName=gStrokeMin}
   Max=ScalarParameter() {defaultValue=0.001, group=Item_Border, parameterName=gStrokeMax}
  ) {materialFunction=&SmoothStep}
  LinearInterpolate(
   A=LinearInterpolate(A=LinearInterpolate(A=@n40, B=@n41.RGB, Alpha=@n42.Result) {constAlpha=0.8}, Alpha=LinearInterpolate(B=@n44.RGB, Alpha=@n47)) {constB=0.025}
   B=@n36
   Alpha=Clamp(None=Multiply(A=@n50.Result))
  )
  Multiply(
   A=@n37.RadialGradientExponential
   B=VectorParameter() {group=Item_Rarity_Colors, parameterName=Rarity 3, defaultValue=<rgba 0 0.2159 0.0015 1>}.RGB
  ) {constB=2}
  ScalarParameter() {defaultValue=1, group=Item_General, parameterName=gAlpha, sliderMax=1}
  TextureCoordinate()
  TextureCoordinate()
  TextureSampleParameter2D() {
   group=Item_Icon
   parameterName=Icon_1
   samplerSource=SSM_Clamp_WorldGroupSettings
   texture=&BonusItemBackground
  }
 ]
}
