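# UI-domain material for a horizontal level meter, filled by the ratio of the CurrentLevel and MaximumLevel parameters.
# An If node compares horizontal UV to that ratio, giving a hard edge between fill and background.
# The fill runs from FillColor to white, with the gradient rescaled to the filled width.
# Beyond the edge, the background blends BackgroundColorA to BackgroundColorB by horizontal UV across the full width.
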Material /FortniteCoreUI/UI/Foundation/Items/Parts/LevelMeterMaterial
graph {
 properties {bUsedWithStaticMesh=false, materialDomain=MD_UI, pixelDepthOffsetMode=PDOM_Legacy, usedShadingModels=Default Lit}
 nodes [
  n1 = TextureCoordinate()
  n2 = TextureCoordinate()
  n3 = Divide(A=ScalarParameter() {defaultValue=2, parameterName=CurrentLevel}, B=ScalarParameter() {defaultValue=5, parameterName=MaximumLevel})
  n4 = LinearInterpolate(
   A=VectorParameter() {parameterName=FillColor, defaultValue=<rgba 0 0.5 0 1>}.RGB
   B=Constant3Vector() {constant=<rgba 1 1 1 1>}
   Alpha=Divide(A=MaterialFunctionCall(Float2=@n2) {materialFunction=&BreakOutFloat2Components}.R, B=@n3)
  )
  If(
   A=MaterialFunctionCall(Float2=TextureCoordinate()) {materialFunction=&BreakOutFloat2Components}.R
   B=@n3
   "A > B"=LinearInterpolate(
    A=VectorParameter() {parameterName=BackgroundColorA, defaultValue=<rgba 0.1 0.1 0.1 1>}.RGB
    B=VectorParameter() {parameterName=BackgroundColorB, defaultValue=<rgba 0.4 0.4 0.4 1>}.RGB
    Alpha=MaterialFunctionCall(Float2=@n1) {materialFunction=&BreakOutFloat2Components}.R
   )
   "A == B"=@n4
   "A < B"=@n4
  )
 ]
}
